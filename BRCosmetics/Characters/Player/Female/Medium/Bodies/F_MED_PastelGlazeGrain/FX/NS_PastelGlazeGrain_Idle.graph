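# Idle Niagara effect for the female PastelGlazeGrain character cosmetic: flapping butterflies, glows, pixie dust and sparkles around the body.
# Butterflies, glows and sparkles sample sockets, while pixie dust spawns from butterfly particle positions.
# System AudioScale gates extra spawning alongside particle-count checks, and modulates brightness, scale and noise.
# VariantID blends time-of-day tints, and bisFootHidden changes butterfly socket choice and glow/sparkle visibility tags.
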NiagaraSystem /BRCosmetics/Characters/Player/Female/Medium/Bodies/F_MED_PastelGlazeGrain/FX/NS_PastelGlazeGrain_Idle
dependencies {
 usedDataInterfaces [
  &NiagaraDataInterfaceArrayInt32
  &NiagaraDataInterfaceCurve
  &NiagaraDataInterfaceVectorCurve
  &NiagaraDataInterfaceParticleRead
  &NiagaraDataInterfaceVector2DCurve
  &NiagaraDataInterfaceSocketReader
  &NiagaraDataInterfaceArrayColor
 ]
 usedDynamicInputs [
  &RandomRangeVector
  &SelectLinearColorFromArray
  &MakeIntFromFloat
  &RemapRange
 ]
 usedModules [
  &self:SystemSpawnScript.SystemScriptSource.SystemScriptGraph.NiagaraNodeAssignment_0.SetVariables_#1
  &SystemState
  &self:SystemSpawnScript.SystemScriptSource.SystemScriptGraph.NiagaraNodeAssignment_2.SetVariables_#2
  &self:SystemSpawnScript.SystemScriptSource.SystemScriptGraph.NiagaraNodeAssignment_1.SetVariables_#3
 ]
}
emitters {
 Butterflies {
  emitterUpdateScript [
   &EmitterState(
    Inactive Response=::NewEnumerator1
    Life Cycle Mode=::NewEnumerator1
    Loop Behavior=::NewEnumerator1
    Loop Duration Mode=::NewEnumerator1
    UseLoopDelay=false
   )
   &SpawnPerFrame(
    Spawn:NiagaraBool=&Boolean_AndOperation(
     Bool1=&SetBoolByFloatComparison(A=$System.AudioScale, B=0.2)
     Bool2=&SetBoolByIntComparison(A=$Engine.Emitter.NumParticles, B=7, Comparison Type=::NewEnumerator6)
    )
    Use Spawn Probability=false
   )
  ]
  particleSpawnScript [
   /SetVariables(Particles.dmp:Vector4f=<1 1 1 1>) {isSetParameters=true}
   &PartitionParticles(Number Of Partitions=9)
   &InitializeParticle(
    Color Maximum=$System.ColorB
    Color Minimum=$System.ColorA
    Color Mode=::NewEnumerator2
    Lifetime Max=1.5
    Lifetime Mode=::NewEnumerator1
    Mesh Scale Mode=::NewEnumerator1
    Mesh Uniform Scale Max=1.5
    Sprite Size=<15 15>
    Sprite Size Mode=::NewEnumerator4
    UsePositionOffset=false
    Write Material Random=true
   )
   &InitialMeshOrientation(Mesh Axis To Orient=<0 1 0>, Mesh Orientation Mode=::NewEnumerator2, Use Rotation Vector=false)
   &SocketLocation(
    Interpolate Socket Sampling=true
    SamplePosOffset=false
    Socket Index=&MakeCustomIntFromBool(
     Bool=$User.bisFootHidden
     True INT=&SelectIntFromArray(
      Array Sampling Mode=::NewEnumerator1
      Direct Array Index=?
      Int Selection Array=DataInterface() {propertyValues {IntData [2,4,5,6,7,8]}}
      Random Seed=0
     )
     False INT=&RandomRangeInt(Minimum=0, Maximum=8)
    )
    Socket Reader=$System.Socket Reader
    Socket Sampling Mode=::NewEnumerator3
    VelocitySpeedLimit=false
   )
   &AddVelocity(
    Module.Use Falloff Distance=false
    Override Seed=false
    Velocity Mode=::NewEnumerator1
    Velocity Origin=$Output.SocketLocation.SampledPosition
    Velocity Speed=5
   )
   &ShapeLocation(MaskSpawnGroup=false, Non Uniform Scale=<2 2 0.2>, Override Seed=false, Sphere Radius=10, UseOwnerScale=true)
   &CameraOffset(Camera Offset Amount=25)
  ]
  particleUpdateScript [
   &ParticleState()
   &ScaleSpriteSize(
    Scale Sprite Size Mode=::NewEnumerator0
    Uniform Scale Factor=&FloatFromCurve(FloatCurve=DataInterface() {propertyValues {Curve {keys [
         {arriveTangent=6.9922, interpMode=RCIM_Cubic, leaveTangent=6.9922, tangentMode=RCTM_User, time=0, value=0}
         {arriveTangentWeight=0.1067, interpMode=RCIM_Cubic, leaveTangentWeight=0.3202, tangentMode=RCTM_User, time=0.2, value=1}
         {arriveTangentWeight=0.2, interpMode=RCIM_Cubic, leaveTangentWeight=0.0667, tangentMode=RCTM_User, time=0.8, value=1}
         {arriveTangent=-6.1729, tangentMode=RCTM_User, time=1, value=0}
        ]}}})
   )
   &WindForce(
    Frequency=2
    Scale:NiagaraFloat=&RandomRangeFloat(Minimum=1, Maximum=2)
    Wind Speed=&Lerp_Vector(
     A=$Particles.Initial.Velocity
     B=&Multiply_VectorByFloat(
      Vector=&NormalizeVector(Vector To Normalize=$Particles.Initial.Velocity, Safe Normalize=true, Safe Fallback Threshold=0, Safe Fallback Vector=<0 0 0.001>)
      Float=1
     )
     Alpha=&FloatFromCurve(FloatCurve=DataInterface() {propertyValues {Curve {keys [{interpMode=RCIM_Cubic, tangentMode=RCTM_User, time=0.9, value=0}, {interpMode=RCIM_Cubic, tangentMode=RCTM_User, time=1, value=1}]}}})
    )
    Wind Speed Scale=&Lerp_Float(A=1, B=2, Alpha=$Particles.MaterialRandom)
   )
   &Drag(Drag=15, Use Linear Drag=true, Use Rotational Drag=true)
   &InheritSourceMovement(
    Applied Position Falloff Distance=5
    Applied Position Scale=&VectorFromCurve(VectorCurve=DataInterface() {propertyValues {XCurve {keys [<at 0 1>,<at 1 0.8>]}, YCurve {keys [<at 0 1>,<at 1 0.8>]}, ZCurve {keys [<at 0 1>,<at 1 0.8>]}}})
    ApplyForce=false
    ApplyPosition=true
    ApplyPositionFalloff=false
   )
   &CurlNoiseForce(
    Bias Noise=false
    Noise Strength=&Multiply_Float(
     A=&RandomRangeFloat(Minimum=100, Maximum=200)
     B=&Lerp_Float(A=5, B=&RandomRangeFloat(Minimum=5, Maximum=15), Alpha=0)
    )
    Pan Noise=false
   )
   &SolveForcesAndVelocity(Clamp Velocity=false, Limit Acceleration=false)
   /SetVariables(Particles.MiddleSpriteSize:Vector2f=&Vector2DFromFloat(Value=&FloatFromCurve(
      FloatCurve=DataInterface() {propertyValues {Curve {keys [
          {arriveTangent=6.9922, interpMode=RCIM_Cubic, leaveTangent=6.9922, tangentMode=RCTM_User, time=0, value=0}
          {arriveTangentWeight=0.1067, interpMode=RCIM_Cubic, leaveTangentWeight=0.3202, tangentMode=RCTM_User, time=0.2, value=1}
          {arriveTangentWeight=0.2, interpMode=RCIM_Cubic, leaveTangentWeight=0.0667, tangentMode=RCTM_User, time=0.8, value=1}
          {arriveTangent=-6.1729, tangentMode=RCTM_User, time=1, value=0}
         ]}}}
      Scale Curve=10
     ))) {isSetParameters=true}
   &ScaleMeshSize(Scale Factor=&VectorFromFloat(Value=&FloatFromCurve(
      FloatCurve=DataInterface() {propertyValues {Curve {keys [
          {arriveTangent=6.9922, interpMode=RCIM_Cubic, leaveTangent=6.9922, tangentMode=RCTM_User, time=0, value=0}
          {arriveTangentWeight=0.1067, interpMode=RCIM_Cubic, leaveTangentWeight=0.3202, tangentMode=RCTM_User, time=0.2, value=1}
          {arriveTangentWeight=0.2, interpMode=RCIM_Cubic, leaveTangentWeight=0.0667, tangentMode=RCTM_User, time=0.8, value=1}
          {arriveTangent=-6.1729, tangentMode=RCTM_User, time=1, value=0}
         ]}}}
      Scale Curve=&Lerp_Float(A=0.95, B=1.1, Alpha=$System.AudioScale)
     )))
   /Niagara/Modules/Update/Utility/IncrementOverTime(Rate of Change:NiagaraFloat=&Multiply_Float(A=&VectorLength(Vector=$Particles.Velocity), B=0.005))
   /SetVariables(Particles.FlapAnim:NiagaraFloat=&FloatFromCurve(
     FloatCurve=DataInterface() {propertyValues {Curve {keys [
         {interpMode=RCIM_Cubic, tangentMode=RCTM_User, time=0, value=0}
         {arriveTangent=4.1706, interpMode=RCIM_Cubic, leaveTangent=4.1706, tangentMode=RCTM_User, time=0.2049, value=0.6056}
         {
          arriveTangent=-0.2246
          arriveTangentWeight=0.2294
          interpMode=RCIM_Cubic
          leaveTangent=-0.2246
          leaveTangentWeight=0.0573
          tangentMode=RCTM_User
          time=0.45
          value=1
         }
         {arriveTangent=-2.436, interpMode=RCIM_Cubic, leaveTangent=-2.436, tangentMode=RCTM_User, time=0.6842, value=0.3528}
         {tangentMode=RCTM_User, time=1, value=0}
        ]}}}
     CurveIndex=&ModuloFloat(A=$Particles.IncrementOverTime.Counter, B=1)
    )) {isSetParameters=true}
   /SetVariables(Particles.Position=&AddVectorToPosition(
     Position=$Particles.Position
     Vector=&MakeVector(
      X=0
      Y=0
      Z=&Waveform(
       "Waveform [1]"=::NewEnumerator0
       "[1] Frequency"=&RandomRangeFloat(Minimum=1, Maximum=1.5)
       "[1] Amplitude Min/Max"=<-1 1>
       "[1] Scale"=0.6
      )
     )
    )) {isSetParameters=true}
   &DynamicMaterialParameters(
    Index 0 Param 1=&Multiply_Float(
     A=&Lerp_Float(A=1, B=2, Alpha=$System.AudioScale)
     B=&RandomRangeFloat(Minimum=1.25, Maximum=1.5)
    )
    Index 0 Param 2=&FloatFromCurve(
     FloatCurve=DataInterface() {propertyValues {Curve {keys [
         {arriveTangent=9.0725, interpMode=RCIM_Cubic, leaveTangent=9.0725, tangentMode=RCTM_User, time=0, value=0.3}
         {arriveTangent=-2.5249, interpMode=RCIM_Cubic, leaveTangent=-2.5249, tangentMode=RCTM_User, time=0.1408, value=0.6}
         {arriveTangent=-0.6983, interpMode=RCIM_Cubic, leaveTangent=-0.6983, time=0.7, value=0.35}
         {arriveTangent=-0.0405, interpMode=RCIM_Cubic, leaveTangent=-0.0405, tangentMode=RCTM_User, time=1, value=0}
        ]}}}
     CurveIndex=&ReturnNormalizedExecIndex()
     Scale Curve=&RandomRangeFloat(Minimum=2, Maximum=3.5)
    )
    Index 0 Param 3=&RandomRangeFloat(Minimum=3, Maximum=6)
    Index 0 Param 4=&RandomRangeFloat(Minimum=2, Maximum=3)
    Module.Index0Param0Write=true
    Module.Index0Param1Write=true
    Module.Index0Param2Write=true
    Module.Index0Param3Write=true
    Module.Index1Param0Write=true
    Module.Index1Param1Write=true
    Module.Index1Param2Write=true
    Module.Index1Param3Write=true
    Write Parameter Index 1=true
   ) {module=DynamicMaterialParameters001}
   &ScaleColor(ScaleA=false, ScaleRGB=false)
   /SetVariables(Particles.SpriteColor=&Multiply_LinearColorByFloat(
     Linear Color=&MultiplyLinearColor(
      Linear Color=$System.SpriteColor
      Scale Factor=&MakeVector4_FromLinearColor(LinearColor=&LerpLinearColors(
        StartColor=&FNTimeOfDayColor(
         Night Color=<rgba 0.7717 0.7612 1 1>
         Morning Color=<rgba 0.5247 0.5573 0.4747 1>
         Day Color=<rgba 1 1 1 1>
         Evening Color=<rgba 0.6079 0.5844 0.5042 1>
        )
        EndColor=&FNTimeOfDayColor(
         Night Color=<rgba 1.1313 1.0961 2 1>
         Morning Color=<rgba 0.8 0.7192 0.5153 1>
         Day Color=<rgba 1 1 1 1>
         Evening Color=<rgba 1 0.601 0.5454 1>
        )
        LerpFactor=$User.VariantID
       ))
     )
     Float=&Multiply_Float(
      A=&FNTimeOfDayScalar(Night Scale=0.8, Morning Scale=0.8, Day Scale=0.8, Evening Scale=0.8)
      B=&Waveform(
       "Waveform [1]"=::NewEnumerator12
       "[1] Frequency"=6
       "[1] Amplitude Min/Max"=<0.4 0.8>
       "[1] Scale"=&Lerp_Float(A=0.25, B=1, Alpha=$System.AudioScale)
      )
     )
    )) {isSetParameters=true}
  ]
  rendererClasses [&NiagaraMeshRendererProperties, &NiagaraSpriteRendererProperties]
  renderers [
   {rendererClass=&NiagaraMeshRendererProperties, rendererIndex=0}
   {rendererClass=&NiagaraSpriteRendererProperties, rendererIndex=1}
  ]
 }
 Glow {
  emitterUpdateScript [
   &EmitterState(
    Inactive Response=::NewEnumerator1
    Life Cycle Mode=::NewEnumerator1
    Loop Behavior=::NewEnumerator1
    Loop Duration Mode=::NewEnumerator1
    UseLoopDelay=false
   )
   &SpawnBurst_Instantaneous(Spawn Count=9, Use Spawn Probability=false, UseLoopCountLimit=false)
  ]
  particleSpawnScript [
   &InitializeParticle(
    Color Maximum=$System.ColorB
    Color Minimum=$System.ColorA
    Color Mode=::NewEnumerator2
    Sprite Size Mode=::NewEnumerator3
    Uniform Sprite Size=50
    UsePositionOffset=false
    Write Material Random=true
   )
   &PartitionParticles(ID to Partition:NiagaraInt32=&ReturnExecIndex(), Number Of Partitions=9)
   &CameraOffset(Camera Offset Amount=15)
  ]
  particleUpdateScript [
   &ParticleState(Kill Particles When Lifetime Has Elapsed=false, Let Infinitely Lived Particles Die When Emitter Deactivates=true)
   &SocketLocation(
    Interpolate Socket Sampling=true
    SamplePosOffset=false
    Socket Index=$Particles.PartitionParticles.Partition
    Socket Reader=$System.Socket Reader
    Socket Sampling Mode=::NewEnumerator3
    VelocitySpeedLimit=false
   )
   &ScaleColor(
    Scale Alpha=&Lerp_Float(A=0.05, B=&RandomRangeFloat(Minimum=0.075, Maximum=0.15), Alpha=$System.AudioScale)
    Scale RGB=&MakeVectorFromLinearColorRGB(Color=&LerpLinearColors(
      StartColor=&FNTimeOfDayColor(
       Night Color=<rgba 0.6877 0.5852 1 1>
       Morning Color=<rgba 0.8 0.7192 0.5153 1>
       Day Color=<rgba 1 1 1 1>
       Evening Color=<rgba 0.6496 0.5407 0.4967 1>
      )
      EndColor=&FNTimeOfDayColor(
       Night Color=<rgba 1.1313 1.0961 2 1>
       Morning Color=<rgba 0.8 0.7192 0.5153 1>
       Day Color=<rgba 1.6905 1.7763 2 1>
       Evening Color=<rgba 1 0.601 0.5454 1>
      )
      LerpFactor=$User.VariantID
     ))
    ScaleA=true
    ScaleRGB=true
   )
   /SetVariables(Particles.VisibilityTag=&MakeCustomIntFromBool(
     Bool=$User.bisFootHidden
     True INT=&SelectIntFromArray(
      Direct Array Index=$Particles.PartitionParticles.Partition
      Int Selection Array=DataInterface() {propertyValues {IntData [1,1,0,1,0,0,0,0,0]}}
     )
     False INT=0
    )) {isSetParameters=true, module=SetVariables_#4}
  ]
  rendererClasses [&NiagaraSpriteRendererProperties]
  renderers [{rendererClass=&NiagaraSpriteRendererProperties}]
 }
 PixieDust {
  emitterUpdateScript [
   &EmitterState(Life Cycle Mode=::NewEnumerator1, Loop Behavior=::NewEnumerator1, Loop Duration Mode=::NewEnumerator1, UseLoopDelay=false)
   &SpawnRate(SpawnRate=40, Use Spawn Probability=false)
   &SpawnPerFrame(
    Spawn:NiagaraBool=&Boolean_AndOperation(
     Bool1=&SetBoolByFloatComparison(A=$System.AudioScale, B=0.2)
     Bool2=&SetBoolByIntComparison(A=$Engine.Emitter.NumParticles, B=12, Comparison Type=::NewEnumerator6)
    )
    Spawn Count:NiagaraInt32=&RandomRangeInt(Minimum=1, Maximum=2)
    Use Spawn Probability=false
   )
  ]
  particleSpawnScript [
   &FN_SetSpawnGroup(Number of Spawn Groups=3)
   &InitializeParticle(
    Color Maximum=$System.ColorB
    Color Minimum=$System.ColorA
    Color Mode=::NewEnumerator2
    Lifetime=&UniformRangedFloat(Minimum=0.5, Maximum=0.75)
    Sprite Rotation Mode=::NewEnumerator2
    Sprite Size Mode=::NewEnumerator1
    Uniform Sprite Size Max=25
    Uniform Sprite Size Min=10
    UsePositionOffset=false
    Write Material Random=true
   )
   &self:SamplePositionFromOtherParticle(PAR=DataInterface() {propertyValues {EmitterBinding {bindingMode=Other, emitterName=Butterflies}}}) {module=SamplePositionFromOtherParticle}
   &DynamicMaterialParameters(
    Index 0 Param 1=&RandomRangeFloat(Minimum=-0.05, Maximum=0.05)
    Index 0 Param 2=&RandomRangeFloat(Minimum=0, Maximum=32)
    Module.Index0Param0Write=true
    Module.Index0Param1Write=true
    Module.Index0Param2Write=false
    Module.Index0Param3Write=false
   )
  ]
  particleUpdateScript [
   &ParticleState()
   &InheritSourceMovement(Applied Position Scale=<0.75 0.75 0.75>, ApplyForce=false, ApplyPosition=true, ApplyPositionFalloff=false)
   &SolveForcesAndVelocity(Clamp Velocity=false, Limit Acceleration=false)
   &ScaleColor(
    Scale Alpha=&FloatFromCurve(
     FloatCurve=DataInterface() {propertyValues {Curve {keys [
         {arriveTangent=9.0725, interpMode=RCIM_Cubic, leaveTangent=9.0725, tangentMode=RCTM_User, time=0, value=0}
         {arriveTangent=-0.5016, interpMode=RCIM_Cubic, leaveTangent=-0.5016, tangentMode=RCTM_User, time=0.1408, value=0.9847}
         {arriveTangent=-0.0405, interpMode=RCIM_Cubic, leaveTangent=-0.0405, tangentMode=RCTM_User, time=1, value=0}
        ]}}}
     Scale Curve=&Lerp_Float(A=0.25, B=0.75, Alpha=$System.AudioScale)
    )
    Scale RGB=&MakeVectorFromLinearColorRGB(Color=&LerpLinearColors(
      StartColor=&FNTimeOfDayColor(
       Night Color=<rgba 0.3971 0.4638 1.25 1>
       Morning Color=<rgba 0.5247 0.5573 0.4747 1>
       Day Color=<rgba 1 1 1 1>
       Evening Color=<rgba 0.6079 0.5844 0.5042 1>
      )
      EndColor=&FNTimeOfDayColor(
       Night Color=<rgba 0.6353 0.7421 2 1>
       Morning Color=<rgba 0.8 0.7192 0.5153 1>
       Day Color=<rgba 1 1 1 1>
       Evening Color=<rgba 1 0.601 0.5454 1>
      )
      LerpFactor=$User.VariantID
     ))
    ScaleA=true
    ScaleRGB=true
   )
   &ScaleSpriteSize(Uniform Curve Sprite Scale=DataInterface() {propertyValues {Curve {keys [
        {interpMode=RCIM_Cubic, tangentMode=RCTM_User, time=0, value=1}
        {arriveTangent=-1.6461, interpMode=RCIM_Cubic, leaveTangent=-1.6461, tangentMode=RCTM_User, time=1, value=0.5}
       ]}}})
  ]
  rendererClasses [&NiagaraSpriteRendererProperties]
  renderers [{rendererClass=&NiagaraSpriteRendererProperties}]
 }
 Sparkles {
  emitterUpdateScript [
   &EmitterState(Life Cycle Mode=::NewEnumerator1, Loop Behavior=::NewEnumerator1, Loop Duration Mode=::NewEnumerator1, UseLoopDelay=false)
   &SpawnRate(SpawnRate=9, Use Spawn Probability=false)
   &SpawnPerFrame(
    Spawn:NiagaraBool=&Boolean_AndOperation(
     Bool1=&SetBoolByFloatComparison(A=$System.AudioScale, B=0.2)
     Bool2=&SetBoolByIntComparison(A=$Engine.Emitter.NumParticles, B=12, Comparison Type=::NewEnumerator6)
    )
    Use Spawn Probability=false
   )
  ]
  particleSpawnScript [
   &PartitionParticles(Number Of Partitions=4)
   &InitializeParticle(
    Color Maximum=$System.ColorB
    Color Minimum=$System.ColorA
    Color Mode=::NewEnumerator2
    Lifetime Max=1
    Lifetime Min=0.5
    Lifetime Mode=::NewEnumerator1
    Sprite Size Mode=::NewEnumerator1
    Uniform Sprite Size Max=2
    Uniform Sprite Size Min=1.5
    UsePositionOffset=false
    Write Material Random=true
   )
   &SocketLocation(
    Interpolate Socket Sampling=true
    SamplePosOffset=false
    Socket Index=$Particles.PartitionParticles.Partition
    Socket Reader=$System.Socket Reader
    Socket Sampling Mode=::NewEnumerator3
    VelocitySpeedLimit=false
   )
   &AddVelocity(
    Module.Use Falloff Distance=false
    Override Seed=false
    Velocity Mode=::NewEnumerator1
    Velocity Origin=$Output.SocketLocation.SampledPosition
    Velocity Speed=&RandomRangeFloat(Minimum=5, Maximum=15)
   )
   &ShapeLocation(MaskSpawnGroup=false, Non Uniform Scale=<2 2 0.2>, Override Seed=false, Sphere Radius=6, UseOwnerScale=true)
  ]
  particleUpdateScript [
   &Drag(Drag=15, Use Linear Drag=true, Use Rotational Drag=true)
   &InheritSourceMovement(
    Applied Position Scale=&VectorFromCurve(VectorCurve=DataInterface() {propertyValues {XCurve {keys [<at 0.7 1>,<at 1 0>]}, YCurve {keys [<at 0.7 1>,<at 1 0>]}, ZCurve {keys [<at 0.7 1>,<at 1 0>]}}})
    ApplyForce=false
    ApplyPosition=true
    ApplyPositionFalloff=false
   )
   &CurlNoiseForce(
    Bias Noise=false
    Noise Frequency=150
    Noise Strength=&Multiply_Float(
     A=&RandomRangeFloat(Minimum=100, Maximum=200)
     B=&Lerp_Float(A=5, B=&RandomRangeFloat(Minimum=5, Maximum=15), Alpha=$System.AudioScale)
    )
    Pan Noise=false
   )
   &SolveForcesAndVelocity(Clamp Velocity=false, Limit Acceleration=false)
   &ParticleState()
   &DynamicMaterialParameters(
    Index 0 Param 1=1
    Index 0 Param 2=&RandomRangeFloat(Minimum=0.25, Maximum=0.75)
    Index 0 Param 3=5
    Index 0 Param 4=&RandomRangeFloat(Minimum=0, Maximum=1)
    Module.Index0Param0Write=true
    Module.Index0Param1Write=true
    Module.Index0Param2Write=true
    Module.Index0Param3Write=true
    Module.Index1Param0Write=true
    Module.Index1Param1Write=true
    Module.Index1Param2Write=true
    Module.Index1Param3Write=true
    Write Parameter Index 1=true
   )
   &ScaleColor(
    Scale RGB=&MakeVectorFromLinearColorRGB(Color=&LerpLinearColors(
      StartColor=&FNTimeOfDayColor(
       Night Color=<rgba 0.1818 0.2123 0.5722 1>
       Morning Color=<rgba 0.5247 0.5573 0.4747 1>
       Day Color=<rgba 1 1 1 1>
       Evening Color=<rgba 0.6079 0.5844 0.5042 1>
      )
      EndColor=&FNTimeOfDayColor(
       Night Color=<rgba 0.2687 0.3139 0.846 1>
       Morning Color=<rgba 0.8 0.7192 0.5153 1>
       Day Color=<rgba 1 1 1 1>
       Evening Color=<rgba 1 0.601 0.5454 1>
      )
      LerpFactor=$User.VariantID
     ))
    ScaleA=false
    ScaleRGB=true
   )
   /SetVariables(Particles.VisibilityTag=&MakeCustomIntFromBool(
     Bool=$User.bisFootHidden
     True INT=&SelectIntFromArray(Direct Array Index=$Particles.PartitionParticles.Partition, Int Selection Array=DataInterface() {propertyValues {IntData [1,1,0,1]}})
     False INT=0
    )) {isSetParameters=true, module=SetVariables_#4}
  ]
  rendererClasses [&NiagaraSpriteRendererProperties]
  renderers [{rendererClass=&NiagaraSpriteRendererProperties}]
 }
}
summary {
 systemName=NS_PastelGlazeGrain_Idle
 emitters [
  {emitterName=Glow, rendererClasses [&NiagaraSpriteRendererProperties]}
  {emitterName=Butterflies, rendererClasses [&NiagaraMeshRendererProperties, &NiagaraSpriteRendererProperties]}
  {emitterName=PixieDust, rendererClasses [&NiagaraSpriteRendererProperties]}
  {emitterName=Sparkles, rendererClasses [&NiagaraSpriteRendererProperties]}
 ]
 userVariables [
  User.bisFootHidden {type=NiagaraBool}
  User.bIsFrontEnd {type=NiagaraBool}
  User.bIsFrontEndPreview {type=NiagaraBool}
  User.bKillReaction {type=NiagaraBool}
  User.VariantID {type=NiagaraFloat, defaultValue=0, description=""}
 ]
}
