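# Material function for the PastelGlazeGrain character cosmetic that layers audio-reactive emissive and animated butterfly meshes onto incoming material attributes.
# Audio amplitude and four spectrum bands, weighted by a packed RGBA reactive map, form reactivity signals alongside time-of-day-scaled emissive color.
# The IsButterflyMesh? static switch enables vertex-shader wing flapping, blended normals and adjusted roughness.
# Helper chains build wing pivots, per-vertex scale variation, sine wobble, triplanar noise and spectrum selection.
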MaterialFunction /BRCosmetics/Characters/Player/Female/Medium/Bodies/F_MED_PastelGlazeGrain/Materials/MF_PastelGlazeGrain
graph {nodes [
  n1 = GetMaterialAttributes(None=FunctionInput() {inputType=FunctionInput_MaterialAttributes}) {attributeGetTypes [
    #1
    #2
    #3
    #4
   ]}
  n2 = MaterialFunctionCall(BaseNormal=@n1.Normal, AdditionalNormal=NamedRerouteUsage()) {materialFunction=&BlendAngleCorrectedNormals}
  n3 = NamedRerouteUsage()
  n4 = NamedRerouteUsage()
  n5 = NamedRerouteUsage()
  n6 = NamedRerouteUsage()
  n7 = Add(A=Power(Base=NamedRerouteUsage()) {constExponent=3}, B=Multiply(A=NamedRerouteUsage()))
  n8 = VectorParameter() {group=PastelGlazeGrain, parameterName=EmissiveColor, defaultValue=<rgba 1 0.1593 0.6341 1>}
  n9 = Add(A=TextureCoordinate() {coordinateIndex=1}, B=Constant2Vector() {r=-1})
  n10 = Constant() {r=-0.5}
  n11 = LinearInterpolate(A=Multiply(A=@n8.RGB) {constB=1.5}, B=@n8.RGB, Alpha=OneMinus(None=SphereMask(A=@n9, B=@n10) {attenuationRadius=0.5, hardnessPercent=0.1})) {constA=1}
  n12 = VectorParameter() {group=PastelGlazeGrain, parameterName=EmissiveIntensity_TOD, defaultValue=<rgba 2 6 2 2>}
  n13 = MaterialFunctionCall(Value_Morning=@n12.R, Value_Day=@n12.R, Value_Evening=@n12.R, Value_Night=@n12.R) {materialFunction=&Scalar_Time_of_Day}
  n14 = Multiply(A=Multiply(A=Multiply(A=@n5, B=@n6), B=Multiply(A=@n7, B=@n11)), B=@n13.Result) {constB=4}
  n15 = NamedRerouteUsage()
  n16 = NamedRerouteUsage()
  n17 = Multiply(A=Multiply(A=@n8.RGB, B=Multiply(A=@n15, B=@n16) {constB=10}), B=@n13.Result) {constB=10}
  n18 = Multiply(
   A=@n17
   B=LinearInterpolate(
    A=ScalarParameter() {defaultValue=1, parameterName=BodyReactiveEmissiveMin}
    B=ScalarParameter() {defaultValue=2.5, parameterName=BodyReactiveEmissiveMax}
    Alpha=NamedRerouteUsage()
   ) {constA=0.5, constB=1.5}
  )
  n19 = LinearInterpolate(B=@n1.Roughness, Alpha=NamedRerouteUsage()) {constA=1}
  n20 = NamedRerouteUsage()
  n21 = SetMaterialAttributes!(
   MaterialAttributes=@n1.MaterialAttributes
   Normal=StaticSwitchParameter(True=LinearInterpolate(A=@n1.Normal, B=@n2.Result, Alpha=@n3), False=@n1.Normal) {group=PastelGlazeGrain, parameterName=IsButterflyMesh?}
   WorldPositionOffset=StaticSwitchParameter(True=Multiply(A=@n4, B=@n3), False=@n1.WorldPositionOffset) {group=PastelGlazeGrain, parameterName=IsButterflyMesh?}
   EmissiveColor=Multiply(A=Add(A=@n14, B=@n18), B=NamedRerouteUsage())
   Roughness=StaticSwitchParameter(True=LinearInterpolate(A=@n1.Roughness, B=@n19, Alpha=@n20), False=@n1.Roughness) {group=PastelGlazeGrain, parameterName=IsButterflyMesh?}
  ) {attributeSetTypes [
    #1
    #2
    #5
    #4
   ]}
  FunctionOutput(None=@n21) {outputName=OUT}
  n22 = DotProduct(A=CameraVectorWS(), B=VertexNormalWS() {desc=Model Space Normal})
  Multiply(A=@n17, B=Power(Base=ConstantBiasScale(None=@n22)) {constExponent=4})
  n23 = MaterialFunctionCall() {materialFunction=/Game/Packages/Fortress_SharedMaterials/Base_Material_Functions/BaseCharacterShaderFunctions/Audio/MF_AudioReactivity}
  n24 = NamedRerouteUsage()
  n25 = Multiply(A=@n23.Amplitude, B=@n24)
  NamedRerouteDeclaration(None=Step(Y=@n25) {constX=0.1, constY=0.1}) {name=AudioTriggered, nodeColor=<rgba 1 0 0.1641 1>}
  n26 = LinearInterpolate(Alpha=ComponentMask(None=NamedRerouteUsage()) {b=true}) {constA=0.25}
  n27 = VectorParameter() {parameterName=GradientRange, channelNames {a=Range4, b=Range3, g=Range2, r=Range1}, defaultValue=<rgba 0.25 0.5 0.75 0.9999>}
  n28 = Step(Y=@n26, X=@n27.Range1) {constX=0.4, constY=0.25}
  n29 = Multiply(A=@n28, B=NamedRerouteUsage())
  n30 = Step(Y=@n26, X=@n27.Range2) {constX=0.5, constY=0.25}
  n31 = NamedRerouteUsage()
  n32 = OneMinus(None=@n30)
  n33 = Step(Y=@n26, X=@n27.Range3) {constX=0.75, constY=0.25}
  n34 = Multiply(A=@n33, B=NamedRerouteUsage())
  n35 = OneMinus(None=@n33)
  n36 = Multiply(A=Step(Y=@n26, X=@n27.Range4) {constX=0.9999, constY=0.25}, B=NamedRerouteUsage())
  n37 = Add(A=Multiply(A=OneMinus(None=@n28), B=Multiply(A=@n30, B=@n31)), B=Add(A=Multiply(A=@n32, B=@n34), B=Multiply(A=@n35, B=@n36)) {constB=0}) {constB=0}
  NamedRerouteDeclaration(None=Saturate(None=Add(A=@n29, B=@n37) {constB=0})) {name=SelectedSpectrum, nodeColor=<rgba 0 0.625 1 1>}
  NamedRerouteDeclaration(None=ScalarParameter() {defaultValue=1, group=PastelGlazeGrain, parameterName=EnableAudioReactivity}) {name=ReactivityOn, nodeColor=<rgba 0.1719 0 1 1>}
  n38 = Constant3Vector() {desc=Random, constant=<rgba -0.3864 0.6 0.4464 0>}
  n39 = ScalarParameter() {defaultValue=4, group=PastelGlazeGrain, parameterName=Wobble_Speed, sortPriority=1}
  n40 = Multiply(A=NamedRerouteUsage()) {constB=2}
  n41 = Add(A=Multiply(A=Time(), B=Add(A=@n39, B=@n40)), B=ScalarParameter() {group=PastelGlazeGrain, parameterName=Wobble_Offset, sortPriority=2})
  n42 = MaterialFunctionCall(Float3=LocalPosition().XYZ) {materialFunction=&BreakOutFloat3Components}
  n43 = ScalarParameter() {defaultValue=10, group=PastelGlazeGrain, parameterName=Wobble_Scale, sortPriority=0}
  n44 = Divide(A=@n42.R, B=@n43)
  n45 = Sine(None=Multiply(A=@n38.R, B=Add(A=@n41, B=@n44))) {period=6.283}
  n46 = Divide(A=@n42.G, B=@n43)
  n47 = Sine(None=Multiply(A=@n38.G, B=Add(A=@n41, B=@n46))) {period=6.283}
  n48 = Divide(A=@n42.B, B=@n43)
  n49 = Sine(None=Multiply(A=@n38.B, B=Add(A=@n41, B=@n48))) {period=6.283}
  n50 = ScalarParameter() {defaultValue=1, group=PastelGlazeGrain, parameterName=Wobble_Amplitude, sortPriority=4}
  n51 = VectorParameter() {desc=Global input when possible, group=WPO_Wobble, parameterName=WindVector, defaultValue=<rgba 0.1 0.1 0 1>}
  n52 = Clamp(None=Length(None=@n51.RGB) {is=class MaterialExpressionLength})
  n53 = MaterialFunctionCall(Float3=Multiply(
    A=MaterialFunctionCall(X=@n45, Y=@n47, Z=@n49) {materialFunction=&MakeFloat3}.Result
    B=Multiply(A=@n50, B=@n52) {desc=Amp}
   ) {desc=wind factor}) {materialFunction=&BreakOutFloat3Components}
  n54 = Multiply(
   A=MaterialFunctionCall(X=Add(A=@n53.G, B=@n53.G), Y=Add(A=@n53.R, B=@n53.R), Z=Add(A=@n53.R, B=@n53.R)) {materialFunction=&MakeFloat3}.Result
   B=Multiply(A=ScalarParameter() {defaultValue=1, group=PastelGlazeGrain, parameterName=Wobble_Strength, sortPriority=3}) {constB=0.1, desc=adjustStrength}
  )
  n55 = VertexNormalWS()
  NamedRerouteDeclaration(None=Multiply(A=Multiply(A=@n54, B=@n55), B=NamedRerouteUsage())) {name=Wobble, nodeColor=<rgba 1 0 0.6797 1>}
  n56 = LocalPosition() {localOrigin=InstancePreSkinning}
  n57 = Multiply(A=Divide(A=VertexInterpolator(VS=@n56.XYZ).PS, B=ObjectBounds()) {constB=128}, B=Constant3Vector() {constant=<rgba 2 2 1 0>})
  n58 = MaterialFunctionCall(X=Multiply(A=Time()) {constB=0.5}) {materialFunction=&MakeFloat2}
  n59 = TextureObject() {samplerType=SAMPLERTYPE_Grayscale, texture=/Game/Global/Textures/Default/Noise/LowResBlurredNoiseNormalized}
  n60 = TextureSample(UVs=Add(A=ComponentMask(None=@n57) {b=true, r=true}, B=@n58.Result), Tex=@n59) {samplerSource=SSM_Wrap_WorldGroupSettings}
  n61 = TextureSample(UVs=Add(A=ComponentMask(None=@n57) {b=true, g=true}, B=@n58.Result), Tex=@n59) {samplerSource=SSM_Wrap_WorldGroupSettings}
  n62 = FunctionInput(Preview=VertexNormalWS()) {bUsePreviewValueAsDefault=true, inputName=World Space Normal, previewValue=<0 0 1 1>}
  n63 = ComponentMask(None=@n62) {r=true}
  n64 = Constant() {r=1}
  n65 = ComponentMask(None=MaterialFunctionCall(In=Abs(None=@n63), Contrast=@n64) {materialFunction=&CheapContrast}.Result) {r=true}
  n66 = Add(A=ComponentMask(None=@n57) {g=true, r=true}, B=@n58.Result)
  n67 = MaterialFunctionCall(In=Abs(None=ComponentMask(None=@n62) {b=true}), Contrast=@n64) {materialFunction=&CheapContrast}
  NamedRerouteDeclaration(None=LinearInterpolate(
    A=LinearInterpolate(A=@n60.RGB, B=@n61.RGB, Alpha=@n65) {desc=Float3}
    B=TextureSample(UVs=@n66, Tex=@n59) {samplerSource=SSM_Wrap_WorldGroupSettings}.RGB
    Alpha=ComponentMask(None=@n67.Result) {r=true}
   )) {name=TriplannarNoise, nodeColor=<rgba 1 0 0.1641 1>}
  n68 = Constant() {r=-0.25}
  n69 = ConstantBiasScale(None=TextureCoordinate()) {bias=-0.5, scale=2}
  n70 = Add(
   A=Constant3Vector() {constant=<rgba -0.3128 0.928 0 0>}
   B=AppendVector(
    A=MaterialFunctionCall!(Angle=@n68, UV/2DVector=@n69) {materialFunction=/Game/Characters/Player/Male/Medium/Bodies/M_MED_Faux/Materials/MF_2DRotate}.Result
    B=Constant()
   )
  )
  n71 = Abs(None=@n70)
  NamedRerouteDeclaration(None=Multiply(A=ComponentMask(None=@n71) {r=true}, B=ScalarParameter() {defaultValue=5, parameterName=Scale})) {name=WobbleMask, nodeColor=<rgba 0.2656 0 1 1>}
  n72 = VertexNormalWS()
  n73 = RotateAboutAxis(NormalizedRotationAxis=NamedRerouteUsage(), RotationAngle=NamedRerouteUsage(), PivotPoint=Constant(), Position=@n72)
  NamedRerouteDeclaration(None=Transform(None=Add(A=@n73, B=@n72)) {transformSourceType=TRANSFORMSOURCE_World, transformType=TRANSFORM_Tangent}) {name=Normal, nodeColor=<rgba 0 1 0.2266 1>}
  n74 = TextureCoordinate() {coordinateIndex=1}
  NamedRerouteDeclaration(None=Step(Y=ComponentMask(None=@n74) {r=true}) {constX=0.5}) {name=LeftRightMask, nodeColor=<rgba 1 0 0.4922 1>}
  n75 = LinearInterpolate(
   A=ScalarParameter() {defaultValue=0.5, group=PastelGlazeGrain, parameterName=SpeedMin}
   B=ScalarParameter() {defaultValue=2, group=PastelGlazeGrain, parameterName=SpeedMax}
   Alpha=ComponentMask(None=NamedRerouteUsage()) {b=true}
  )
  n76 = Time()
  n77 = LinearInterpolate(
   A=Multiply(A=@n76, B=ScalarParameter() {defaultValue=0.4, parameterName=ButterflyReactiveWingflapSpeed}) {constB=0.09}
   B=Multiply(A=@n76) {constB=0.05}
   Alpha=NamedRerouteUsage()
  )
  n78 = Logarithm!(None=Frac(None=Multiply(A=@n75, B=@n77)))
  n79 = ScalarParameter() {defaultValue=0.1, parameterName=ButterflyReactiveWingRotationMin}
  n80 = ScalarParameter() {defaultValue=0.25, parameterName=ButterflyReactiveWingRotationMax}
  n81 = NamedRerouteUsage()
  n82 = Multiply(
   A=ConstantBiasScale(None=Sine(None=@n78))
   B=LinearInterpolate(
    A=LinearInterpolate(A=@n79, B=@n80, Alpha=@n81)
    B=ScalarParameter() {defaultValue=0.25, group=PastelGlazeGrain, parameterName=RotationAngle}
    Alpha=NamedRerouteUsage()
   ) {constA=0.125}
  ) {constA=1}
  NamedRerouteDeclaration(None=LinearInterpolate(A=Multiply(A=@n82) {constB=-1}, B=@n82, Alpha=NamedRerouteUsage())) {name=Rotation, nodeColor=<rgba 1 0.8672 0 1>}
  NamedRerouteDeclaration(None=WorldPosition().XYZ) {name=WorldPos, nodeColor=<rgba 0 0.0156 1 1>}
  n83 = ComponentMask(None=TextureCoordinate() {coordinateIndex=1}) {r=true}
  n84 = OneMinus(None=@n83)
  n85 = OneMinus(None=Power(Base=Min(A=@n83, B=@n84)) {constExponent=1})
  n86 = Multiply(
   A=Transform(None=Constant3Vector() {constant=<rgba 0 1 0 0>})
   B=Multiply(
    A=Multiply(B=@n85) {constA=2.204}
    B=MaterialFunctionCall() {materialFunction=&ObjectScale}.Scale XYZ
   )
  )
  n87 = Multiply(A=@n86) {constB=-1}
  n88 = NamedRerouteUsage()
  NamedRerouteDeclaration(None=Add(A=NamedRerouteUsage(), B=LinearInterpolate(A=@n87, B=@n86, Alpha=@n88))) {name=Pivot, nodeColor=<rgba 0.3359 1 0 1>}
  n89 = Subtract(A=ComponentMask(None=@n74) {r=true}) {constB=0.5}
  NamedRerouteDeclaration(None=Step(X=Abs(None=@n89)) {constX=0.45, constY=0.4}) {name=EdgeMask, nodeColor=<rgba 0 1 0.0625 1>}
  NamedRerouteDeclaration(None=Transform(None=Constant3Vector() {constant=<rgba -1 0 0 0>})) {name=Axis, nodeColor=<rgba 1 0 0.2578 1>}
  n90 = ScalarParameter() {defaultValue=-0.5, group=PastelGlazeGrain, parameterName=Clip Offset}
  n91 = Subtract(A=@n90)
  n92 = ComponentMask(None=NamedRerouteUsage()) {r=true}
  n93 = MaterialFunctionCall() {materialFunction=&CameraDirectionVector}
  n94 = DotProduct(A=@n93.Result, B=VertexNormalWS())
  n95 = Multiply(A=@n94, B=@n94)
  n96 = Multiply(A=Add(A=LinearInterpolate(A=@n91, B=@n90, Alpha=@n92), B=OneMinus(None=@n95)), B=@n93.Result)
  n97 = NamedRerouteUsage()
  n98 = NamedRerouteUsage()
  n99 = NamedRerouteUsage()
  n100 = NamedRerouteUsage()
  n101 = NamedRerouteUsage()
  n102 = NamedRerouteDeclaration(None=LinearInterpolate(
    A=ScalarParameter() {defaultValue=-1, group=PastelGlazeGrain, parameterName=ScaleMin}
    B=ScalarParameter() {defaultValue=1, group=PastelGlazeGrain, parameterName=ScaleMax}
    Alpha=ComponentMask(None=@n101) {r=true}
   )) {name=Scale, nodeColor=<rgba 0.5234 0 1 1>}
  n103 = Add(
   A=Multiply(A=NamedRerouteUsage(), B=RotateAboutAxis(NormalizedRotationAxis=@n97, RotationAngle=@n98, PivotPoint=@n99, Position=@n100))
   B=Multiply(A=NamedRerouteUsage(), B=@n102)
  ) {constB=0}
  NamedRerouteDeclaration(None=Add(A=Add(A=@n96, B=@n103), B=NamedRerouteUsage())) {name=WPO, nodeColor=<rgba 0.5 1 0 1>}
  n104 = MaterialFunctionCall(UVs=TextureCoordinate() {coordinateIndex=1}, Rotation Center=Constant(), "Rotation Angle (0-1)"=Constant() {r=0.25}) {desc=rotate90, materialFunction=&CustomRotator}
  n105 = Constant()
  n106 = AppendVector(A=Subtract(A=AppendVector(A=@n104.Rotated Values, B=@n105)) {constB=0.5}, B=Constant())
  NamedRerouteDeclaration(None=Multiply(
    A=Transform(None=@n106)
    B=MaterialFunctionCall() {materialFunction=&ObjectScale}.Scale XYZ
   )) {name=CenterVector, nodeColor=<rgba 0 1 0.3906 1>}
  NamedRerouteDeclaration(None=VertexColor()) {name=VertColor, nodeColor=<rgba 0 0.3438 1 1>}
  NamedRerouteDeclaration(None=Multiply(A=NamedRerouteUsage(), B=Add(A=@n102))) {name=ScaledCenterVector, nodeColor=<rgba 0 1 0.4609 1>}
  n107 = Constant()
  n108 = ScalarParameter() {defaultValue=0.1, group=PastelGlazeGrain, parameterName=SphereMaskRadius}
  NamedRerouteDeclaration(None=Saturate(None=SphereMask(A=@n70, B=@n107, Radius=@n108) {attenuationRadius=0.25})) {name=WPOMask, nodeColor=<rgba 0.1484 0 1 1>}
  NamedRerouteDeclaration(None=@n1.BaseColor) {name=BaseColor, nodeColor=<rgba 0 0.5312 1 1>}
  n109 = Power(Base=@n25, Exp=ScalarParameter() {defaultValue=3, group=PastelGlazeGrain, parameterName=Audio_Amplitude_Exp})
  NamedRerouteDeclaration(None=@n109) {name=OverallAmp, nodeColor=<rgba 1 0 0.9609 1>}
  NamedRerouteDeclaration(None=SmoothStep(Value=@n25) {constMax=0.9, constMin=0.1}) {name=AudioOnTransition, nodeColor=<rgba 1 0 0.1641 1>}
  n110 = TextureSampleParameter2D() {
   group=PastelGlazeGrain
   parameterName=AudioReactiveMap
   texture=/BRCosmetics/Characters/Player/Female/Medium/Bodies/F_MED_PastelGlazeGrain/Textures/T_PastelGlazeGrain_ReactiveMap
  }
  n111 = Power(Base=Multiply(A=@n23.Spectrum, B=@n24), Exp=ScalarParameter() {defaultValue=5, group=PastelGlazeGrain, parameterName=Audio_Spectrum_Exp}) {constExponent=4}
  n112 = MaterialFunctionCall(Float4=Add(A=Multiply(A=@n109, B=@n111) {constB=0.05}, B=@n111)) {desc=Bass/Low/Mid/High, materialFunction=&BreakOutFloat4Components}
  n113 = VectorParameter() {group=PastelGlazeGrain, parameterName=Spectrum_Intensity, defaultValue=<rgba 2 4 6 8>}
  n114 = Multiply(A=@n112.R, B=@n113.R)
  n115 = Multiply(A=@n112.G, B=@n113.G)
  n116 = Add(
   A=Multiply(A=@n110.R, B=NamedRerouteDeclaration(None=@n114) {name=BassFreq, nodeColor=<rgba 1 0.4821 0 1>}) {constB=0}
   B=Multiply(A=@n110.G, B=NamedRerouteDeclaration(None=@n115) {name=LowFreq, nodeColor=<rgba 1 0.4821 0 1>}) {constB=0}
  )
  n117 = Multiply(A=@n110.B, B=NamedRerouteDeclaration(None=Multiply(A=@n112.B, B=@n113.B)) {name=MidFreq, nodeColor=<rgba 1 0.4821 0 1>}) {constB=0}
  n118 = NamedRerouteDeclaration(None=Multiply(A=@n112.A, B=@n113.A)) {name=HighFreq, nodeColor=<rgba 1 0.4821 0 1>}
  NamedRerouteDeclaration(None=Add(A=Add(A=@n116, B=@n117), B=Multiply(A=@n110.A, B=@n118) {constB=0}) {constB=0}) {name=AudioReactiveMap, nodeColor=<rgba 1 0.1875 0 1>}
  OneMinus(None=@n22)
 ]}
