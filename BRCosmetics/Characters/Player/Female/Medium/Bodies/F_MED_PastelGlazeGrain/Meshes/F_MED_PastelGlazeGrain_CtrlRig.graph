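# Control Rig for the F_MED_PastelGlazeGrain female body, driving corrective and secondary bones for skirt, legs and feet.
# Skirt bases follow the thighs via rotation constraints, and thigh pose readers weight further skirt offsets, with the stronger side driving the front center.
# A right-calf pose reader weights bf_ corrections, calf twist translations reset to initial, and foot correctives follow constraints and remapped foot yaw.
# A second sequence branch runs the shared DeformRig function.
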ControlRigBlueprint /BRCosmetics/Characters/Player/Female/Medium/Bodies/F_MED_PastelGlazeGrain/Meshes/F_MED_PastelGlazeGrain_CtrlRig
graphs [
 RigVMFunctionLibrary {links [], nodes []}
 RigVMModel {nodes [
   RigUnit_BeginExecution = &RigUnit_BeginExecution()
   RigVMFunction_Sequence = &RigVMFunction_Sequence(ExecuteContext=@RigUnit_BeginExecution.ExecutePin)
   &RigUnit_RotationConstraintLocalSpaceOffset(
    AdvancedSettings=(InterpolationType=Average,RotationOrderForFilter=XZY)
    Child=(Type=Bone,Name="dyn_skirt_fr_a_base_l")
    Filter=(bX=True,bY=True,bZ=True)
    Parents=((Item=(Type=Bone,Name="thigh_l"),Weight=1))
    Weight=1
    ExecutePin=@RigVMFunction_Sequence.A
   )
   > &RigUnit_RotationConstraintLocalSpaceOffset(
    AdvancedSettings=(InterpolationType=Average,RotationOrderForFilter=XZY)
    Child=(Type=Bone,Name="dyn_skirt_fr_a_base_r")
    Filter=(bX=True,bY=True,bZ=True)
    Parents=((Item=(Type=Bone,Name="thigh_r"),Weight=1))
    Weight=1
   )
   > &RigUnit_RotationConstraintLocalSpaceOffset(
    AdvancedSettings=(InterpolationType=Average,RotationOrderForFilter=XZY)
    Child=(Type=Bone,Name="dyn_skirt_side_a_base_l")
    Filter=(bX=True,bY=True,bZ=True)
    Parents=((Item=(Type=Bone,Name="thigh_l"),Weight=1))
    Weight=0.8
   )
   > &RigUnit_RotationConstraintLocalSpaceOffset(
    AdvancedSettings=(InterpolationType=Average,RotationOrderForFilter=XZY)
    Child=(Type=Bone,Name="dyn_skirt_side_a_base_r")
    Filter=(bX=True,bY=True,bZ=True)
    Parents=((Item=(Type=Bone,Name="thigh_r"),Weight=1))
    Weight=0.8
   )
   > n1 = &RigUnit_SphericalPoseReader(
    ActiveRegionScaleFactors=(PositiveWidth=1,NegativeWidth=1,PositiveHeight=1,NegativeHeight=1)
    ActiveRegionSize=0.06
    Debug=(bDrawDebug=true,DebugScale=25,DebugSegments=20,DebugThickness=0.25)
    DriverAxis=(X=-1,Y=0,Z=0)
    DriverItem=(Type=Bone,Name="thigh_l")
    FalloffRegionScaleFactors=(PositiveWidth=1,NegativeWidth=1,PositiveHeight=1,NegativeHeight=1)
    FalloffSize=0.5
    OptionalParentItem=(Type=Bone)
    RotationOffset=(X=100,Y=0,Z=-20)
   )
   > n2 = &RigUnit_SphericalPoseReader(
    ActiveRegionScaleFactors=(PositiveWidth=1,NegativeWidth=1,PositiveHeight=1,NegativeHeight=1)
    ActiveRegionSize=0.06
    Debug=(bDrawDebug=true,DebugScale=25,DebugSegments=20,DebugThickness=0.25)
    DriverAxis=(X=1,Y=0,Z=0)
    DriverItem=(Type=Bone,Name="thigh_r")
    FalloffRegionScaleFactors=(PositiveWidth=1,NegativeWidth=1,PositiveHeight=1,NegativeHeight=1)
    FalloffSize=0.5
    OptionalParentItem=(Type=Bone)
    RotationOffset=(X=-80,Y=0,Z=-20)
   )
   > &RigUnit_ModifyTransforms(
    ItemToModify=((Item=(Type=Bone,Name="dyn_skirt_fr_a_base"),Transform=(Rotation=(X=0,Y=0,Z=0.6428,W=0.766),Translation=(X=-5,Y=0,Z=0),Scale3D=(X=1,Y=1,Z=1))))
    Weight=RigVMDispatchNode(
     Condition=/Script/RigVM.RigVMFunction_MathDoubleGreater(A=@n2.OutputParam, B=@n1.OutputParam) {is=struct RigVMUnitNode, template="Greater::Execute(in A,in B,out Result)"}.Result
     True=@n2.OutputParam
     False=@n1.OutputParam
    ) {template="DISPATCH_RigVMDispatch_If(in Condition,in True,in False,out Result)"}.Result
   )
   > &RigUnit_ModifyTransforms(
    ItemToModify="((Item=(Type=Bone,Name=\"dyn_skirt_side_a_base_l\"),Transform=(Rotation=(X=0,Y=0,Z=-0.0872,W=0.9962),Translation=(X=-3.8864,Y=-10,Z=-3.6),Scale3D=(X=1,Y=1,Z=1))),(Item=(Type=Bone,Name=\"dyn_skirt_fr_a_base_l\"),Transform=(Rotation=(X=0,Y=-0.0436,Z=0,W=0.999),T ...(truncated)"
    Weight=@n1.OutputParam
   )
   > &RigUnit_ModifyTransforms(
    ItemToModify="((Item=(Type=Bone,Name=\"dyn_skirt_side_a_base_r\"),Transform=(Rotation=(X=0,Y=0,Z=0,W=1),Translation=(X=0,Y=4.4404,Z=0),Scale3D=(X=1,Y=1,Z=1))),(Item=(Type=Bone,Name=\"dyn_skirt_fr_a_base_r\"),Transform=(Rotation=(X=0,Y=0,Z=0,W=1),Translation=(X=0,Y=5,Z=0),Sc ...(truncated)"
    Weight=@n2.OutputParam
   )
   > n3 = &RigUnit_SphericalPoseReader(
    ActiveRegionScaleFactors=(PositiveWidth=1,NegativeWidth=1,PositiveHeight=1,NegativeHeight=1)
    ActiveRegionSize=0.1
    Debug=(bDrawDebug=true,bDraw2D=False,bDrawLocalAxes=False,DebugScale=25,DebugSegments=20,DebugThickness=0.25)
    DriverAxis=(X=1,Y=0,Z=0)
    DriverItem=(Type=Bone,Name="calf_r")
    FalloffRegionScaleFactors=(PositiveWidth=1,NegativeWidth=1,PositiveHeight=1,NegativeHeight=1)
    FalloffSize=0.2
    FlipHeightScaling=False
    FlipWidthScaling=False
    OptionalParentItem=(Type=Bone,Name="None")
    RotationOffset=(X=90,Y=0,Z=60)
   )
   > &RigUnit_ModifyTransforms(
    ItemToModify="((Item=(Type=Bone,Name=\"bf_correction\"),Transform=(Rotation=(X=0,Y=0,Z=0,W=1),Translation=(X=3,Y=0,Z=0),Scale3D=(X=1,Y=1,Z=1))),(Item=(Type=Bone,Name=\"bf_calf_correction\"),Transform=(Rotation=(X=0,Y=0,Z=0,W=1),Translation=(X=1,Y=0,Z=0),Scale3D=(X=1,Y=1,Z=1 ...(truncated)"
    WeightMinimum=0
    Weight=@n3.OutputParam
   )
   > Set Transform = /Script/ControlRig.RigUnit_SetTranslation(
    Item=(Type=Bone,Name="calf_twist_01_l")
    Space=LocalSpace
    Weight=1
    bInitial=False
    bPropagateToChildren=True
    Value=&RigUnit_GetTransform(Item=(Type=Bone,Name="calf_twist_01_l"), Space=LocalSpace, bInitial=true).Transform.Translation
   ) {is=struct RigVMUnitNode, template="Set Transform::Execute(in Item,in Space,in bInitial,in Value,in Weight,in bPropagateToChildren)"}
   > Set Transform_1 = &RigUnit_SetTranslation(
    Item=(Type=Bone,Name="calf_twist_01_r")
    Space=LocalSpace
    Weight=1
    bInitial=False
    bPropagateToChildren=True
    Value=&RigUnit_GetTransform(Item=(Type=Bone,Name="calf_twist_01_r"), Space=LocalSpace, bInitial=true).Transform.Translation
   ) {template="Set Transform::Execute(in Item,in Space,in bInitial,in Value,in Weight,in bPropagateToChildren)"}
   > &RigUnit_RotationConstraintLocalSpaceOffset(
    AdvancedSettings=(InterpolationType=Average,RotationOrderForFilter=XZY)
    Child=(Type=Bone,Name="ball_pos_corr_l")
    Filter=(bX=True,bY=True,bZ=True)
    Parents=((Item=(Type=Bone,Name="ball_l"),Weight=1))
    Weight=1
   )
   > &RigUnit_RotationConstraintLocalSpaceOffset(
    AdvancedSettings=(InterpolationType=Average,RotationOrderForFilter=XZY)
    Child=(Type=Bone,Name="ball_pos_corr_r")
    Filter=(bX=True,bY=True,bZ=True)
    Parents=((Item=(Type=Bone,Name="ball_r"),Weight=1))
    Weight=1
   )
   > &RigUnit_ParentConstraint(
    AdvancedSettings=(InterpolationType=Average,RotationOrderForFilter=XZY)
    Child=(Type=Bone,Name="foot_pos_corr_l")
    Filter=(TranslationFilter=(bX=true,bY=True,bZ=True),RotationFilter=(bX=True,bY=true,bZ=True),ScaleFilter=(bX=True,bY=True,bZ=True))
    Parents=((Item=(Type=Bone,Name="foot_l"),Weight=1))
    Weight=1
    bMaintainOffset=true
   )
   > &RigUnit_ParentConstraint(
    AdvancedSettings=(InterpolationType=Average,RotationOrderForFilter=XZY)
    Child=(Type=Bone,Name="foot_pos_corr_r")
    Filter=(TranslationFilter=(bX=true,bY=true,bZ=true),RotationFilter=(bX=True,bY=true,bZ=True),ScaleFilter=(bX=True,bY=True,bZ=True))
    Parents=((Item=(Type=Bone,Name="foot_r"),Weight=1))
    Weight=1
    bMaintainOffset=true
   )
   > Get Delta Ref_8 = RigVMFunctionReferenceNode(Item=(Type=Bone,Name="foot_l"))
   > &RigUnit_ModifyTransforms(
    ItemToModify="((Item=(Type=Bone,Name=\"foot_corr_bk_l\"),Transform=(Rotation=(X=-0.3502,Y=0,Z=0,W=0.9367),Translation=(X=0,Y=-4.2,Z=0.671),Scale3D=(X=1,Y=1,Z=1))),(Item=(Type=Bone,Name=\"None\"),Transform=(Rotation=(X=0,Y=0,Z=0,W=1),Translation=(X=0,Y=0,Z=0),Scale3D=(X=1,Y= ...(truncated)"
    WeightMinimum=0
    Weight=&RigVMFunction_MathDoubleRemap(SourceMaximum=90, SourceMinimum=0, TargetMaximum=1, TargetMinimum=0, Value=@"Get Delta Ref_8".Result.Rotation.Yaw).Result
   )
   > Get Delta Ref_8_1 = RigVMFunctionReferenceNode(Item=(Type=Bone,Name="foot_r"))
   > &RigUnit_ModifyTransforms(
    ItemToModify="((Item=(Type=Bone,Name=\"foot_corr_bk_r\"),Transform=(Rotation=(X=-0.3502,Y=0,Z=0,W=0.9367),Translation=(X=0,Y=4.2,Z=-0.671),Scale3D=(X=1,Y=1,Z=1))),(Item=(Type=Bone,Name=\"None\"),Transform=(Rotation=(X=0,Y=0,Z=0,W=1),Translation=(X=0,Y=0,Z=0),Scale3D=(X=1,Y= ...(truncated)"
    WeightMinimum=0
    Weight=&RigVMFunction_MathDoubleRemap(SourceMaximum=90, SourceMinimum=0, TargetMaximum=1, TargetMinimum=0, Value=@"Get Delta Ref_8_1".Result.Rotation.Yaw).Result
   )
   > Get Delta Ref_8_2 = RigVMFunctionReferenceNode(Item=(Type=Bone,Name="foot_l"))
   > &RigUnit_ModifyTransforms(
    ItemToModify="((Item=(Type=Bone,Name=\"foot_corr_fw_l\"),Transform=(Rotation=(X=0.3214,Y=0.0036,Z=0.0615,W=0.9449),Translation=(X=-0.0235,Y=2.4552,Z=1.0313),Scale3D=(X=1,Y=1,Z=0.8))),(Item=(Type=Bone,Name=\"None\"),Transform=(Rotation=(X=0,Y=0,Z=0,W=1),Translation=(X=0,Y=0, ...(truncated)"
    WeightMinimum=0
    Weight=&RigVMFunction_MathDoubleRemap(SourceMaximum=90, SourceMinimum=0, TargetMaximum=1, TargetMinimum=0, Value=@"Get Delta Ref_8_2".Result.Rotation.Yaw).Result
   )
   > &RigUnit_ModifyTransforms(
    ItemToModify="((Item=(Type=Bone,Name=\"foot_corr_fw_l\"),Transform=(Rotation=(X=-0.0498,Y=0,Z=0,W=0.9988),Translation=(X=0,Y=-0.5854,Z=2.4958),Scale3D=(X=1,Y=1,Z=0.8103))),(Item=(Type=Bone,Name=\"None\"),Transform=(Rotation=(X=0,Y=0,Z=0,W=1),Translation=(X=0,Y=0,Z=0),Scale3 ...(truncated)"
    WeightMinimum=0
    Weight=&RigVMFunction_MathDoubleRemap(SourceMaximum=-45, SourceMinimum=0, TargetMaximum=1, TargetMinimum=0, Value=@"Get Delta Ref_8_2".Result.Rotation.Yaw).Result
   )
   > Get Delta Ref_8_1_1 = RigVMFunctionReferenceNode(Item=(Type=Bone,Name="foot_r"))
   > &RigUnit_ModifyTransforms(
    ItemToModify="((Item=(Type=Bone,Name=\"foot_corr_fw_r\"),Transform=(Rotation=(X=0.0713,Y=0.0005,Z=-0.0018,W=0.9975),Translation=(X=0.0014,Y=-1.1564,Z=1.2228),Scale3D=(X=1,Y=1,Z=1))),(Item=(Type=Bone,Name=\"None\"),Transform=(Rotation=(X=0,Y=0,Z=0,W=1),Translation=(X=0,Y=0,Z ...(truncated)"
    WeightMinimum=0
    Weight=&RigVMFunction_MathDoubleRemap(SourceMaximum=90, SourceMinimum=0, TargetMaximum=1, TargetMinimum=0, Value=@"Get Delta Ref_8_1_1".Result.Rotation.Yaw).Result
   )
   > &RigUnit_ModifyTransforms(
    ItemToModify="((Item=(Type=Bone,Name=\"foot_corr_fw_r\"),Transform=(Rotation=(X=-0.0498,Y=0,Z=0,W=0.9988),Translation=(X=0,Y=0.5854,Z=-2.4958),Scale3D=(X=1,Y=1,Z=0.8103))),(Item=(Type=Bone,Name=\"None\"),Transform=(Rotation=(X=0,Y=0,Z=0,W=1),Translation=(X=0,Y=0,Z=0),Scale3 ...(truncated)"
    WeightMinimum=0
    Weight=&RigVMFunction_MathDoubleRemap(SourceMaximum=-45, SourceMinimum=0, TargetMaximum=1, TargetMinimum=0, Value=@"Get Delta Ref_8_1_1".Result.Rotation.Yaw).Result
   )
   FN_DeformRig_FNC = RigVMFunctionReferenceNode(
    DeformRig_ Curve="(deformRig_l_upperarm_up_alpha_72_#1=False,deformRig_r_upperarm_up_alpha_73_#2=False,deformRig_l_upperarm_fr_alpha_74_#3=False,deformRig_r_upperarm_fr_alpha_75_CB51B8 ...(truncated)"
    DeformRig_Bool="(upperarm_1_#4=True,elbow_3_#5=True,wrist_7_#6=True,groin_8_#7=True,glute_17_#8=True,knee_9_40C408324CA261 ...(truncated)"
    DeformRig_Debug=(upperarm_debug_18_#4=False,wrist_debug_20_#6=False,lat_debug_26_#9=False,groinGlute_debug_27_#7=False)
    DeformRig_Value="(upperarm_up_18_#10=(Rotation=(X=0,Y=0,Z=0,W=1),Translation=(X=0,Y=0,Z=8),Scale3D=(X=0.7,Y=1,Z=1)),upperarm_fwd_fr_19_#11=(Rotation=(X=0,Y=0,Z=0,W=1),Translation=(X=0,Y=10,Z=0),Scale3D=(X=0.5,Y=1,Z= ...(truncated)"
    ExecuteContext=@RigVMFunction_Sequence.B
   )
   KalmanFilter = &RigVMFunction_KalmanFloat(BufferSize=8, Value=0) {template="KalmanFilter::Execute(in Value,in BufferSize,out Result)"}
   &RigUnit_ParentConstraint(
    AdvancedSettings=(InterpolationType=Average,RotationOrderForFilter=XZY)
    Child=(Type=Bone,Name="ball_pos_corr_l")
    ExecutePin=()
    Filter=(TranslationFilter=(bX=false,bY=false,bZ=false),RotationFilter=(bX=true,bY=true,bZ=true),ScaleFilter=(bX=true,bY=true,bZ=true))
    Parents=((Item=(Type=Bone,Name="ball_l"),Weight=1))
    Weight=1
    bMaintainOffset=true
   )
   &RigUnit_ParentConstraint(
    AdvancedSettings=(InterpolationType=Average,RotationOrderForFilter=XZY)
    Child=(Type=Bone,Name="ball_pos_corr_r")
    ExecutePin=()
    Filter=(TranslationFilter=(bX=false,bY=false,bZ=false),RotationFilter=(bX=true,bY=true,bZ=true),ScaleFilter=(bX=true,bY=true,bZ=true))
    Parents=((Item=(Type=Bone,Name="ball_r"),Weight=1))
    Weight=1
    bMaintainOffset=true
   )
   RigUnit_GetRelativeTransformForItem = &RigUnit_GetRelativeTransformForItem(Child=(Type=Bone,Name="calf_twist_01_l"), Parent=(Type=Bone,Name="calf_twist_01_l"), bChildInitial=False, bParentInitial=False)
  ]}
]
hierarchy [
 root {loc=<0 0 0>, rot=<0 0 0 1>}
 pelvis {parent=root, loc=<0 -0.5301 98.0571>, rot=<0.0127 0.707 -0.0127 -0.707>}
 spine_01 {parent=pelvis, loc=<6.1208 -0.9328 0>, rot=<0 0 0.034 -0.9994>}
 spine_02 {parent=spine_01, loc=<6.5996 -0.2095 0>, rot=<0 0 0.0238 -0.9997>}
 spine_03 {parent=spine_02, loc=<6.4914 -0.1237 0>, rot=<0 0 0.0149 -0.9999>}
 spine_04 {parent=spine_03, loc=<6.9598 0.0794 0>, rot=<0 0 -0.0132 -0.9999>}
 spine_05 {parent=spine_04, loc=<6.9987 0 0>, rot=<0 0 -0.0015 -1>}
 attach_cape {parent=spine_05, loc=<9.017 -11.714 0>, rot=<0.0616 -0.7044 -0.0616 -0.7044>}
 neck_01 {parent=spine_05, loc=<11.3138 0.1529 0>, rot=<0 0 -0.2003 -0.9797>}
 neck_02 {parent=neck_01, loc=<4.5627 0 0>, rot=<0 0 0.0504 -0.9987>}
 head {parent=neck_02, loc=<5.3914 0 0>, rot=<0 0 0.115 -0.9934>}
 dyn_hair_bk_MSTR {parent=head, loc=<11.6835 -5.2363 0>, rot=<0.1215 -0.6966 0.6966 -0.1215>}
 dyn_hair_bk_00 {parent=dyn_hair_bk_MSTR, loc=<0.5 0 0>, rot=<0 -0.0572 0 -0.9984>}
 dyn_hair_bk_01 {parent=dyn_hair_bk_00, loc=<6.0938 0 0>, rot=<0 -0.0272 0 -0.9996>}
 dyn_hair_bk_02 {parent=dyn_hair_bk_01, loc=<8.7854 0 0>, rot=<0 0.1593 0 -0.9872>}
 dyn_hair_bk_03 {parent=dyn_hair_bk_02, loc=<9.4823 0 0>, rot=<0 -0.1683 0 -0.9857>}
 dyn_hair_bk_04 {parent=dyn_hair_bk_03, loc=<10.3536 0 0>, rot=<0 -0.0751 0 -0.9972>}
 dyn_hair_bk_05 {parent=dyn_hair_bk_04, loc=<8.7412 0 0>, rot=<0 0 0 1>}
 dyn_hair_l_MSTR {parent=head, loc=<8.8541 0.6916 -6.499>, rot=<0.0262 -0.7919 0.5984 -0.1187>}
 dyn_hair_l_00 {parent=dyn_hair_l_MSTR, loc=<0.5 0 0>, rot=<0 0 0 1>}
 dyn_hair_l_01 {parent=dyn_hair_l_00, loc=<7.0495 0 0>, rot=<0.0202 0.1166 0.0179 -0.9928>}
 dyn_hair_l_02 {parent=dyn_hair_l_01, loc=<7.1616 0 0>, rot=<-0.086 0.2331 0.0422 -0.9677>}
 dyn_hair_l_03 {parent=dyn_hair_l_02, loc=<8.7592 0 0>, rot=<-0.0205 -0.2538 0.0767 -0.964>}
 dyn_hair_l_04 {parent=dyn_hair_l_03, loc=<8.993 0 0>, rot=<0.0802 -0.1932 -0.017 -0.9777>}
 dyn_hair_r_MSTR {parent=head, loc=<10.0791 -1.0364 6.0327>, rot=<0.1315 -0.5972 0.7911 0.0152>}
 dyn_hair_r_00 {parent=dyn_hair_r_MSTR, loc=<0.5 0 0>, rot=<-0.0006 0.0016 -0.0049 -1>}
 dyn_hair_r_01 {parent=dyn_hair_r_00, loc=<6.1151 0 0>, rot=<-0.0063 0.0224 -0.0723 -0.9971>}
 dyn_hair_r_02 {parent=dyn_hair_r_01, loc=<7.6839 0 0>, rot=<-0.0459 0.3353 -0.0465 -0.9398>}
 dyn_hair_r_03 {parent=dyn_hair_r_02, loc=<7.9495 0 0>, rot=<0.1523 -0.2716 -0.0403 -0.9494>}
 dyn_hair_r_04 {parent=dyn_hair_r_03, loc=<8.1126 0 0>, rot=<0.0022 -0.1842 -0.0117 -0.9828>}
 dyn_hair_r_05 {parent=dyn_hair_r_04, loc=<13.7322 0 0>, rot=<0 0 0 1>}
 FACIAL_C_FacialRoot {parent=head, loc=<2.1524 -0.746 -0.0002>, rot=<0.4879 -0.5118 -0.4879 -0.5118>}
 FACIAL_C_Forehead {parent=FACIAL_C_FacialRoot, loc=<-0.0032 -2.3496 9.2228>, rot=<-0.0475 0.0002 -0.0001 -0.9989>}
 FACIAL_L_ForeheadIn {parent=FACIAL_C_FacialRoot, loc=<1.3459 -2.2842 9.1137>, rot=<-0.0648 -0.0694 0.0038 -0.9955>}
 FACIAL_R_ForeheadIn {parent=FACIAL_C_FacialRoot, loc=<-1.368 -2.3006 9.1069>, rot=<-0.0648 0.0699 -0.0041 -0.9954>}
 FACIAL_L_ForeheadMid {parent=FACIAL_C_FacialRoot, loc=<2.6562 -2.507 8.6657>, rot=<-0.075 -0.1753 -0.0215 -0.9814>}
 FACIAL_R_ForeheadMid {parent=FACIAL_C_FacialRoot, loc=<-2.6798 -2.5433 8.6526>, rot=<-0.0751 0.1758 0.0212 -0.9813>}
 FACIAL_L_ForeheadOut {parent=FACIAL_C_FacialRoot, loc=<3.8697 -2.3579 7.9257>, rot=<-0.1253 -0.2551 -0.0804 -0.9554>}
 FACIAL_R_ForeheadOut {parent=FACIAL_C_FacialRoot, loc=<-3.9126 -2.4159 7.8728>, rot=<-0.1254 0.2555 0.08 -0.9553>}
 FACIAL_L_EyesackUpper {parent=FACIAL_C_FacialRoot, loc=<2.6094 -1.5484 9.1379>, rot=<0.0448 -0.0837 -0.065 -0.9934>}
 FACIAL_R_EyesackUpper {parent=FACIAL_C_FacialRoot, loc=<-2.6404 -1.5783 9.1917>, rot=<0.0447 0.0842 0.0647 -0.9933>}
 FACIAL_L_EyelidUpperFurrow {parent=FACIAL_C_FacialRoot, loc=<2.4753 -1.3371 9.0034>, rot=<-0.0341 -0.1503 -0.0804 -0.9848>}
 FACIAL_R_EyelidUpperFurrow {parent=FACIAL_C_FacialRoot, loc=<-2.4765 -1.3181 9.0381>, rot=<-0.0341 0.1507 0.0801 -0.9847>}
 FACIAL_L_EyelidUpperA {parent=FACIAL_C_FacialRoot, loc=<2.6078 -0.7057 7.6021>, rot=<-0.1519 0.0002 -0.0002 -0.9884>}
 FACIAL_R_EyelidUpperA {parent=FACIAL_C_FacialRoot, loc=<-2.6079 -0.706 7.5976>, rot=<-0.1519 0.0002 -0.0002 -0.9884>}
 FACIAL_L_Eye {parent=FACIAL_C_FacialRoot, loc=<2.5958 -0.6806 7.6244>, rot=<-0.0041 0.0002 -0.0001 -1>}
 FACIAL_R_Eye {parent=FACIAL_C_FacialRoot, loc=<-2.5965 -0.6799 7.6234>, rot=<-0.0041 0.0002 -0.0001 -1>}
 FACIAL_L_EyelidLowerA {parent=FACIAL_C_FacialRoot, loc=<2.6078 -0.7057 7.6021>, rot=<0.2801 0.0003 -0.0001 -0.96>}
 FACIAL_R_EyelidLowerA {parent=FACIAL_C_FacialRoot, loc=<-2.6079 -0.706 7.5976>, rot=<0.28 0.0003 -0.0001 -0.96>}
 FACIAL_L_EyeCornerInner {parent=FACIAL_C_FacialRoot, loc=<1.6507 -0.3742 8.5558>, rot=<-0.0041 0.0002 -0.0001 -1>}
 FACIAL_R_EyeCornerInner {parent=FACIAL_C_FacialRoot, loc=<-1.6841 -0.3273 8.5567>, rot=<-0.0041 0.0002 -0.0001 -1>}
 FACIAL_L_EyeCornerOuter {parent=FACIAL_C_FacialRoot, loc=<3.8615 -0.4254 8.1663>, rot=<-0.0038 -0.3824 -0.0017 -0.924>}
 FACIAL_R_EyeCornerOuter {parent=FACIAL_C_FacialRoot, loc=<-3.8994 -0.4289 8.1317>, rot=<-0.0039 0.3828 0.0015 -0.9238>}
 FACIAL_L_EyesackLower {parent=FACIAL_C_FacialRoot, loc=<2.8365 0.8167 7.4987>, rot=<-0.0263 -0.1394 -0.023 -0.9896>}
 FACIAL_R_EyesackLower {parent=FACIAL_C_FacialRoot, loc=<-2.8391 0.8205 7.483>, rot=<-0.0264 0.1399 0.0227 -0.9896>}
 FACIAL_L_CheekInner {parent=FACIAL_C_FacialRoot, loc=<1.9146 1.2882 8.1246>, rot=<-0.1272 -0.087 -0.1976 -0.9681>}
 FACIAL_R_CheekInner {parent=FACIAL_C_FacialRoot, loc=<-1.9143 1.2876 8.1206>, rot=<-0.1273 0.0874 0.1973 -0.9681>}
 FACIAL_L_CheekOuter {parent=FACIAL_C_FacialRoot, loc=<3.5982 1.5397 6.6839>, rot=<0.0514 -0.3771 0.0987 -0.9195>}
 FACIAL_R_CheekOuter {parent=FACIAL_C_FacialRoot, loc=<-3.6261 1.5278 6.6579>, rot=<0.0514 0.3776 -0.0989 -0.9192>}
 FACIAL_L_NasolabialBulge {parent=FACIAL_C_FacialRoot, loc=<3.2124 4.8855 6.6132>, rot=<0.0541 -0.3772 0.0221 -0.9243>}
 FACIAL_R_NasolabialBulge {parent=FACIAL_C_FacialRoot, loc=<-3.2126 4.885 6.6091>, rot=<0.0541 0.3777 -0.0224 -0.9241>}
 FACIAL_L_NasolabialFurrow {parent=FACIAL_C_FacialRoot, loc=<2.7453 3.5701 8.844>, rot=<-0.037 -0.1241 -0.3643 -0.9222>}
 FACIAL_R_NasolabialFurrow {parent=FACIAL_C_FacialRoot, loc=<-2.7446 3.5743 8.8444>, rot=<-0.0372 0.1245 0.364 -0.9223>}
 FACIAL_L_CheekLower {parent=FACIAL_C_FacialRoot, loc=<3.6109 4.5805 5.8824>, rot=<0.0309 -0.4269 0.0559 -0.902>}
 FACIAL_R_CheekLower {parent=FACIAL_C_FacialRoot, loc=<-3.6112 4.5791 5.8778>, rot=<0.0308 0.4273 -0.0561 -0.9018>}
 FACIAL_C_Nose {parent=FACIAL_C_FacialRoot, loc=<-0.0003 0.8667 9.5854>, rot=<-0.1344 0.0002 -0.0002 -0.9909>}
 FACIAL_C_NoseLower {parent=FACIAL_C_Nose, loc=<0.0031 2.6679 -0.1735>, rot=<0.1305 0 0 -0.9914>}
 FACIAL_L_Nostril {parent=FACIAL_C_Nose, loc=<0.9595 1.9861 -0.3583>, rot=<0.0173 0.1305 0.0023 -0.9913>}
 FACIAL_R_Nostril {parent=FACIAL_C_Nose, loc=<-0.978 1.956 -0.3597>, rot=<0.0173 -0.1305 -0.0023 -0.9913>}
 FACIAL_C_NoseTip {parent=FACIAL_C_Nose, loc=<-0.0009 1.8731 0.5693>, rot=<0.0872 0 0 -0.9962>}
 FACIAL_C_MouthUpper {parent=FACIAL_C_FacialRoot, loc=<-0.0004 5.0191 7.5594>, rot=<-0.0041 0.0002 -0.0001 -1>}
 FACIAL_C_LipUpper {parent=FACIAL_C_MouthUpper, loc=<0.0023 -0.018 2.0463>, rot=<-0.0872 0 0 -0.9962>}
 FACIAL_L_LipUpper {parent=FACIAL_C_MouthUpper, loc=<0.9627 0.0198 1.8519>, rot=<-0.0778 -0.1601 -0.0379 -0.9833>}
 FACIAL_R_LipUpper {parent=FACIAL_C_MouthUpper, loc=<-0.963 0.0211 1.853>, rot=<-0.0778 0.1601 0.0379 -0.9833>}
 FACIAL_L_LipUpperOuter {parent=FACIAL_C_MouthUpper, loc=<1.6806 0.175 1.3374>, rot=<-0.0984 -0.2293 -0.0983 -0.9634>}
 FACIAL_R_LipUpperOuter {parent=FACIAL_C_MouthUpper, loc=<-1.6812 0.1787 1.342>, rot=<-0.0984 0.2293 0.0983 -0.9634>}
 FACIAL_L_LipCorner {parent=FACIAL_C_MouthUpper, loc=<2.1846 0.4844 0.7445>, rot=<0 -0.2078 0 -0.9782>}
 FACIAL_R_LipCorner {parent=FACIAL_C_MouthUpper, loc=<-2.1874 0.4865 0.7454>, rot=<0 0.2078 0 -0.9782>}
 FACIAL_L_Masseter {parent=FACIAL_C_FacialRoot, loc=<5.8005 2.2062 4.5879>, rot=<0.0653 0.1131 0.0082 -0.9914>}
 FACIAL_R_Masseter {parent=FACIAL_C_FacialRoot, loc=<-5.8027 2.2058 4.5842>, rot=<0.0653 -0.1126 -0.0085 -0.9915>}
 FACIAL_C_UnderChin {parent=FACIAL_C_FacialRoot, loc=<0.0017 9.1795 4.8777>, rot=<0.4963 0.0003 0 -0.8682>}
 FACIAL_L_UnderChin {parent=FACIAL_C_FacialRoot, loc=<3.1796 7.6778 3.6613>, rot=<0.3991 -0.3549 0.0808 -0.8415>}
 FACIAL_R_UnderChin {parent=FACIAL_C_FacialRoot, loc=<-3.1884 7.6755 3.6645>, rot=<0.399 0.3555 -0.0809 -0.8414>}
 FACIAL_C_TeethUpper {parent=FACIAL_C_FacialRoot, loc=<0.0009 5.5371 8.8502>, rot=<-0.0041 0.0002 -0.0001 -1>}
 FACIAL_C_LowerLipRotation {parent=FACIAL_C_FacialRoot, loc=<0.0002 2.7782 2.7542>, rot=<0.3382 0.0003 0 -0.9411>}
 FACIAL_C_MouthLower {parent=FACIAL_C_LowerLipRotation, loc=<0.0024 -1.3302 5.1325>, rot=<-0.342 0 0 -0.9397>}
 FACIAL_C_LipLower {parent=FACIAL_C_MouthLower, loc=<0.0032 1.1904 1.8234>, rot=<0.0872 0 0 -0.9962>}
 FACIAL_L_LipLower {parent=FACIAL_C_MouthLower, loc=<0.8195 1.1486 1.6458>, rot=<0.0927 -0.1294 0.0272 -0.9869>}
 FACIAL_R_LipLower {parent=FACIAL_C_MouthLower, loc=<-0.8184 1.149 1.6474>, rot=<0.0927 0.1294 -0.0272 -0.9869>}
 FACIAL_L_LipLowerOuter {parent=FACIAL_C_MouthLower, loc=<1.5695 0.9221 1.1961>, rot=<0.0913 -0.2321 0.0318 -0.9679>}
 FACIAL_R_LipLowerOuter {parent=FACIAL_C_MouthLower, loc=<-1.5701 0.9274 1.1995>, rot=<0.0913 0.2321 -0.0318 -0.9679>}
 FACIAL_C_TeethLower {parent=FACIAL_C_LowerLipRotation, loc=<0.0044 -1.7091 6.3864>, rot=<-0.342 0 0 -0.9397>}
 FACIAL_C_Tongue1 {parent=FACIAL_C_TeethLower, loc=<-0.0033 0.0547 -4.7825>, rot=<-0.2588 0 0 -0.9659>}
 FACIAL_C_Tongue2 {parent=FACIAL_C_Tongue1, loc=<0.0003 0.4031 1.3683>, rot=<0.2588 0 0 -0.9659>}
 FACIAL_C_Tongue3 {parent=FACIAL_C_Tongue2, loc=<0.0006 0.2039 1.2524>, rot=<0.1305 0 0 -0.9914>}
 FACIAL_C_Tongue4 {parent=FACIAL_C_Tongue3, loc=<0.0014 -0.1656 1.0846>, rot=<0.0436 0 0 -0.999>}
 FACIAL_C_TongueUpper3 {parent=FACIAL_C_Tongue3, loc=<0.0001 -0.3363 -0.1599>, rot=<0 0 0 1>}
 FACIAL_C_TongueLower3 {parent=FACIAL_C_Tongue3, loc=<0.0001 0.8009 0.0396>, rot=<0 0 0 1>}
 FACIAL_L_TongueSide3 {parent=FACIAL_C_Tongue3, loc=<1.4994 0.0133 0.1446>, rot=<0 0 0 1>}
 FACIAL_R_TongueSide3 {parent=FACIAL_C_Tongue3, loc=<-1.4993 0.0137 0.1441>, rot=<0 0 0 1>}
 FACIAL_C_TongueUpper2 {parent=FACIAL_C_Tongue2, loc=<0.0001 -0.5479 -0.121>, rot=<0 0 0 1>}
 FACIAL_L_TongueSide2 {parent=FACIAL_C_Tongue2, loc=<1.7786 0.0676 0.2363>, rot=<0 0 0 1>}
 FACIAL_R_TongueSide2 {parent=FACIAL_C_Tongue2, loc=<-1.7785 0.068 0.2357>, rot=<0 0 0 1>}
 FACIAL_C_Jaw {parent=FACIAL_C_FacialRoot, loc=<0.0002 2.7782 2.7542>, rot=<0.3382 0.0003 0 -0.9411>}
 FACIAL_C_Jawline {parent=FACIAL_C_Jaw, loc=<0.0045 2.7418 6.8185>, rot=<0.2586 0 0 -0.966>}
 FACIAL_L_ChinSide {parent=FACIAL_C_Jaw, loc=<2.6581 0.5098 6.6611>, rot=<-0.2208 -0.3738 0.0499 -0.8995>}
 FACIAL_R_ChinSide {parent=FACIAL_C_Jaw, loc=<-2.6495 0.5107 6.663>, rot=<-0.2208 0.3738 -0.0499 -0.8995>}
 FACIAL_C_Chin {parent=FACIAL_C_Jaw, loc=<0.0046 1.2226 7.2066>, rot=<-0.2591 0 0 -0.9659>}
 FACIAL_L_Ear {parent=FACIAL_C_FacialRoot, loc=<6.1416 0.4512 1.7647>, rot=<-0.0304 -0.0003 -0.0263 -0.9992>}
 FACIAL_R_Ear {parent=FACIAL_C_FacialRoot, loc=<-6.1417 0.4489 1.7579>, rot=<-0.0304 0.0008 0.026 -0.9992>}
 FACIAL_L_Jawline {parent=FACIAL_C_FacialRoot, loc=<4.282 5.9736 5.0146>, rot=<0.2053 0.1582 -0.0541 -0.9643>}
 FACIAL_R_Jawline {parent=FACIAL_C_FacialRoot, loc=<-4.2858 5.9743 5.0112>, rot=<0.2054 -0.1577 0.0539 -0.9644>}
 FACIAL_C_NoseBridge {parent=FACIAL_C_FacialRoot, loc=<0.0023 -0.9933 9.9091>, rot=<0.0397 0.0003 -0.0001 -0.9992>}
 FACIAL_L_EyelidUpperB {parent=FACIAL_C_FacialRoot, loc=<2.6078 -0.7057 7.6021>, rot=<-0.229 0.0002 -0.0002 -0.9734>}
 FACIAL_L_EyelidLowerB {parent=FACIAL_C_FacialRoot, loc=<2.6078 -0.7056 7.6021>, rot=<0.3134 0.0003 0 -0.9496>}
 FACIAL_R_EyelidLowerB {parent=FACIAL_C_FacialRoot, loc=<-2.6079 -0.706 7.5976>, rot=<0.3217 0.0003 0 -0.9469>}
 FACIAL_R_EyelidUpperB {parent=FACIAL_C_FacialRoot, loc=<-2.6079 -0.706 7.5976>, rot=<-0.229 0.0002 -0.0002 -0.9734>}
 attach_backpack {parent=spine_05, loc=<-0.123 -13.473 0>, rot=<-0.0093 -0.707 0.0093 -0.707>}
 clavicle_l {parent=spine_05, loc=<5.4831 -0.1053 -3.2946>, rot=<0.0465 -0.7289 -0.009 -0.6829>}
 upperarm_l {parent=clavicle_l, loc=<9.8658 -1.2136 -0.4105>, rot=<0.0164 -0.3245 -0.0269 -0.9454>}
 lowerarm_l {parent=upperarm_l, loc=<26.465 -0.8345 0.259>, rot=<0.0036 0.0161 -0.2298 -0.9731>}
 hand_l {parent=lowerarm_l, loc=<22.7355 -0.3203 -0.6105>, rot=<-0.6784 -0.0223 0.0735 -0.7307>}
 thumb_01_l {parent=hand_l, loc=<1.3118 -0.3191 -2.459>, rot=<0.6546 -0.4042 0.0664 -0.6354>}
 thumb_02_l {parent=thumb_01_l, loc=<5.9336 -0.0448 0.0401>, rot=<-0.0037 0.0213 0.298 -0.9543>}
 thumb_03_l {parent=thumb_02_l, loc=<2.2223 0.3682 -0.0907>, rot=<0.0118 -0.013 0.0066 -0.9998>}
 index_metacarpal_l {parent=hand_l, loc=<1.6535 0.0183 -2.96>, rot=<0.0158 -0.0124 -0.0173 -0.9996>}
 index_01_l {parent=index_metacarpal_l, loc=<8.3008 0.0357 -0.0524>, rot=<-0.0044 0.0485 0.0924 -0.9945>}
 index_02_l {parent=index_01_l, loc=<3.3587 -0.3004 0.0452>, rot=<-0.002 0.0027 0.3225 -0.9466>}
 index_03_l {parent=index_02_l, loc=<2.5578 -0.123 -0.0252>, rot=<-0.002 0.0067 0.2558 -0.9667>}
 middle_metacarpal_l {parent=hand_l, loc=<2.1856 -0.1725 -0.5984>, rot=<0.006 0.0168 -0.0329 -0.9993>}
 middle_01_l {parent=middle_metacarpal_l, loc=<7.3275 -0.2447 0.0129>, rot=<-0.0208 0.0042 0.0858 -0.9961>}
 middle_02_l {parent=middle_01_l, loc=<4.0123 0.1401 0.1175>, rot=<-0.0218 0.0258 0.3512 -0.9357>}
 middle_03_l {parent=middle_02_l, loc=<2.6077 0.0534 0.1373>, rot=<0.0082 0.0188 0.1708 -0.9851>}
 ring_metacarpal_l {parent=hand_l, loc=<2.2074 -0.3625 1.3638>, rot=<-0.0284 0.0722 -0.0458 -0.9959>}
 ring_01_l {parent=ring_metacarpal_l, loc=<7.0295 -0.6197 -0.1446>, rot=<-0.0489 -0.0252 0.1892 -0.9804>}
 ring_02_l {parent=ring_01_l, loc=<3.6174 0.4647 -0.1027>, rot=<0 0 0.2806 -0.9598>}
 ring_03_l {parent=ring_02_l, loc=<2.4152 0.114 0.0199>, rot=<0 0 0.1189 -0.9929>}
 pinky_metacarpal_l {parent=hand_l, loc=<1.7616 -0.9753 3.3814>, rot=<-0.0397 0.0814 -0.069 -0.9935>}
 pinky_01_l {parent=pinky_metacarpal_l, loc=<7.1709 -0.4831 -0.1094>, rot=<-0.0654 -0.0035 0.2242 -0.9723>}
 pinky_02_l {parent=pinky_01_l, loc=<2.8274 0.0893 -0.1043>, rot=<0.0037 0.0011 0.2854 -0.9584>}
 pinky_03_l {parent=pinky_02_l, loc=<2.2046 0.0307 -0.0078>, rot=<-0.0002 0.0098 0.1445 -0.9895>}
 deform_wrist_dn_base_l {parent=hand_l, loc=<0 -1.5326 0>, rot=<0 0 0 -1>}
 deform_wrist_dn_l {parent=deform_wrist_dn_base_l, loc=<0 0 0>, rot=<0 0 0 1>}
 deform_wrist_up_base_l {parent=hand_l, loc=<0 2.3009 0>, rot=<0 0 0 -1>}
 deform_wrist_up_l {parent=deform_wrist_up_base_l, loc=<0 0 0>, rot=<0 0 0 1>}
 lowerarm_twist_02_l {parent=lowerarm_l, loc=<10.6707 0 0>, rot=<0 0 0 1>}
 lowerarm_twist_01_l {parent=lowerarm_l, loc=<18.316 0 0>, rot=<0 0 0 1>}
 deform_elbow_in_base_l {parent=lowerarm_l, loc=<0.1834 3.6062 -0.2572>, rot=<0 0 0 -1>}
 deform_elbow_in_l {parent=deform_elbow_in_base_l, loc=<0 0 0>, rot=<0 0 0 1>}
 deform_elbow_out_base_l {parent=lowerarm_l, loc=<0 -4 0>, rot=<0 0 0 1>}
 deform_elbow_out_l {parent=deform_elbow_out_base_l, loc=<0 0 0>, rot=<0 0 0 1>}
 upperarm_twist_01_l {parent=upperarm_l, loc=<7.2135 -0.1134 0>, rot=<0 0 0 1>}
 deform_upperarm_bk_base_l {parent=upperarm_twist_01_l, loc=<-6.4053 -4.3746 0.314>, rot=<0 0 0 -1>}
 deform_upperarm_bk_l {parent=deform_upperarm_bk_base_l, loc=<0 0 0>, rot=<0 0 0 1>}
 deform_upperarm_fr_base_l {parent=upperarm_twist_01_l, loc=<-5.6312 5.092 0.3354>, rot=<0 0 0 -1>}
 deform_upperarm_fr_l {parent=deform_upperarm_fr_base_l, loc=<0 0 0>, rot=<0 0 0 1>}
 deform_upperarm_up_base_l {parent=upperarm_twist_01_l, loc=<-8.891 0.4825 3.9517>, rot=<0 0 0 -1>}
 deform_upperarm_up_l {parent=deform_upperarm_up_base_l, loc=<0 0 0>, rot=<0 0 0 1>}
 upperarm_twist_02_l {parent=upperarm_l, loc=<19.7998 -0.1134 0>, rot=<0 0 0 1>}
 extra_clavicle_l {parent=clavicle_l, loc=<0 0 0>, rot=<0 0 0 1>}
 clavicle_r {parent=spine_05, loc=<5.4831 -0.1053 3.2946>, rot=<0.7289 0.0465 -0.6829 0.009>}
 upperarm_r {parent=clavicle_r, loc=<-9.8658 1.2136 0.4105>, rot=<0.0164 -0.3245 -0.0269 -0.9454>}
 lowerarm_r {parent=upperarm_r, loc=<-26.465 0.8345 -0.259>, rot=<0.0036 0.0161 -0.2298 -0.9731>}
 hand_r {parent=lowerarm_r, loc=<-22.7355 0.3203 0.6106>, rot=<-0.6784 -0.0223 0.0735 -0.7307>}
 thumb_01_r {parent=hand_r, loc=<-1.3118 0.3191 2.459>, rot=<0.6546 -0.4042 0.0664 -0.6354>}
 thumb_02_r {parent=thumb_01_r, loc=<-5.9336 0.0448 -0.0401>, rot=<-0.0037 0.0213 0.298 -0.9543>}
 thumb_03_r {parent=thumb_02_r, loc=<-2.2223 -0.3682 0.0907>, rot=<0.0118 -0.013 0.0066 -0.9998>}
 index_metacarpal_r {parent=hand_r, loc=<-1.6535 -0.0183 2.96>, rot=<0.0158 -0.0124 -0.0173 -0.9996>}
 index_01_r {parent=index_metacarpal_r, loc=<-8.3008 -0.0357 0.0524>, rot=<-0.0044 0.0485 0.0924 -0.9945>}
 index_02_r {parent=index_01_r, loc=<-3.3587 0.3004 -0.0452>, rot=<-0.002 0.0027 0.3225 -0.9466>}
 index_03_r {parent=index_02_r, loc=<-2.5578 0.123 0.0252>, rot=<-0.002 0.0067 0.2558 -0.9667>}
 middle_metacarpal_r {parent=hand_r, loc=<-2.1856 0.1725 0.5984>, rot=<0.006 0.0168 -0.0329 -0.9993>}
 middle_01_r {parent=middle_metacarpal_r, loc=<-7.3275 0.2447 -0.0129>, rot=<-0.0208 0.0042 0.0858 -0.9961>}
 middle_02_r {parent=middle_01_r, loc=<-4.0123 -0.1401 -0.1175>, rot=<-0.0218 0.0258 0.3512 -0.9357>}
 middle_03_r {parent=middle_02_r, loc=<-2.6077 -0.0534 -0.1373>, rot=<0.0082 0.0188 0.1708 -0.9851>}
 ring_metacarpal_r {parent=hand_r, loc=<-2.2074 0.3625 -1.3638>, rot=<-0.0284 0.0722 -0.0458 -0.9959>}
 ring_01_r {parent=ring_metacarpal_r, loc=<-7.0295 0.6197 0.1446>, rot=<-0.0489 -0.0252 0.1892 -0.9804>}
 ring_02_r {parent=ring_01_r, loc=<-3.6174 -0.4647 0.1027>, rot=<0 0 0.2806 -0.9598>}
 ring_03_r {parent=ring_02_r, loc=<-2.4152 -0.114 -0.0199>, rot=<0 0 0.1189 -0.9929>}
 pinky_metacarpal_r {parent=hand_r, loc=<-1.7616 0.9753 -3.3814>, rot=<-0.0397 0.0814 -0.069 -0.9935>}
 pinky_01_r {parent=pinky_metacarpal_r, loc=<-7.1709 0.4831 0.1094>, rot=<-0.0654 -0.0035 0.2242 -0.9723>}
 pinky_02_r {parent=pinky_01_r, loc=<-2.8274 -0.0893 0.1043>, rot=<0.0037 0.0011 0.2854 -0.9584>}
 pinky_03_r {parent=pinky_02_r, loc=<-2.2046 -0.0307 0.0078>, rot=<-0.0002 0.0098 0.1445 -0.9895>}
 deform_wrist_dn_base_r {parent=hand_r, loc=<-0.0001 1.5328 0>, rot=<0 0 0 -1>}
 deform_wrist_dn_r {parent=deform_wrist_dn_base_r, loc=<0 0 0>, rot=<0 0 0 1>}
 deform_wrist_up_base_r {parent=hand_r, loc=<-0.0002 -2.3007 -0.0001>, rot=<0 0 0 -1>}
 deform_wrist_up_r {parent=deform_wrist_up_base_r, loc=<0 0 0>, rot=<0 0 0 1>}
 lowerarm_twist_02_r {parent=lowerarm_r, loc=<-10.6707 0 0>, rot=<0 0 0 -1>}
 lowerarm_twist_01_r {parent=lowerarm_r, loc=<-18.316 0 0>, rot=<0 0 0 -1>}
 deform_elbow_in_base_r {parent=lowerarm_r, loc=<-0.1834 -3.6062 0.2572>, rot=<0 0 0 -1>}
 deform_elbow_in_r {parent=deform_elbow_in_base_r, loc=<0 0 0>, rot=<0 0 0 1>}
 deform_elbow_out_base_r {parent=lowerarm_r, loc=<0 4 -0.0001>, rot=<0 0 0 -1>}
 deform_elbow_out_r {parent=deform_elbow_out_base_r, loc=<0 0 0>, rot=<0 0 0 1>}
 upperarm_twist_01_r {parent=upperarm_r, loc=<-7.2135 0.1134 0>, rot=<0 0 0 -1>}
 deform_upperarm_bk_base_r {parent=upperarm_twist_01_r, loc=<6.4053 4.3746 -0.314>, rot=<0 0 0 -1>}
 deform_upperarm_bk_r {parent=deform_upperarm_bk_base_r, loc=<0 0 0>, rot=<0 0 0 1>}
 deform_upperarm_fr_base_r {parent=upperarm_twist_01_r, loc=<5.6312 -5.092 -0.3354>, rot=<0 0 0 -1>}
 deform_upperarm_fr_r {parent=deform_upperarm_fr_base_r, loc=<0 0 0>, rot=<0 0 0 1>}
 deform_upperarm_up_base_r {parent=upperarm_twist_01_r, loc=<8.891 -0.4825 -3.9517>, rot=<0 0 0 -1>}
 deform_upperarm_up_r {parent=deform_upperarm_up_base_r, loc=<0 0 0>, rot=<0 0 0 1>}
 upperarm_twist_02_r {parent=upperarm_r, loc=<-19.7998 0.1134 0>, rot=<0 0 0 -1>}
 dyn_cape_collider {parent=spine_05, loc=<0 -0.3005 0>, rot=<0.0284 -0.7065 -0.0284 -0.7065>}
 deform_pec_aim_base_l {parent=spine_04, loc=<3.2456 9.3309 -1.2237>, rot=<0.0603 -0.4033 0.3318 -0.8507>}
 deform_pec_aim_l {parent=deform_pec_aim_base_l, loc=<0 0 0>, rot=<0 0 0 1>}
 deform_pec_base_l {parent=deform_pec_aim_l, loc=<9.3665 0 0>, rot=<0 0 0 1>}
 deform_pec_l {parent=deform_pec_base_l, loc=<0 0 0>, rot=<0 0 0 1>}
 deform_pec_aim_base_r {parent=spine_04, loc=<3.2452 9.3309 1.2237>, rot=<-0.4033 -0.0603 0.8507 0.3318>}
 deform_pec_aim_r {parent=deform_pec_aim_base_r, loc=<0 0 0>, rot=<0 0 0 1>}
 deform_pec_base_r {parent=deform_pec_aim_r, loc=<-9.3667 -0.0003 -0.0004>, rot=<0 0 0 1>}
 deform_pec_r {parent=deform_pec_base_r, loc=<0 0 0>, rot=<0 0 0 1>}
 dyn_dressflap {parent=spine_04, loc=<-1.8128 9.8189 -10.4855>, rot=<0.0294 -0.7065 -0.0294 -0.7065>}
 thigh_l {parent=pelvis, loc=<-1.0855 -0.2109 -8.7053>, rot=<0.0009 0.0355 0.0155 -0.9993>}
 calf_l {parent=thigh_l, loc=<-42.5284 -0.0019 -0.7297>, rot=<0.0006 -0.0161 -0.0391 -0.9991>}
 foot_l {parent=calf_l, loc=<-46.7401 -2.5014 -0.592>, rot=<0.0427 -0.0238 0.0353 -0.9982>}
 ball_l {parent=foot_l, loc=<-3.5658 10.8066 -0.1544>, rot=<0.0001 -0.0001 -0.7063 -0.7079>}
 calf_twist_02_l {parent=calf_l, loc=<-19.1546 -1.7943 -0.2638>, rot=<0.0237 0.0095 -0.0035 -0.9997>}
 calf_twist_01_l {parent=calf_l, loc=<-35.8483 -1.9197 -0.5794>, rot=<0.0237 0.0095 -0.0035 -0.9997>}
 foot_corr_bk_l {parent=calf_twist_01_l, loc=<-6.8935 -4.7046 -0.0364>, rot=<-0.0496 -0.727 0.0133 -0.6847>}
 foot_corr_fw_l {parent=calf_twist_01_l, loc=<-3.1357 4.6623 -0.374>, rot=<0.7287 -0.0019 0.6824 0.0581>}
 deform_knee_in_base_l {parent=calf_l, loc=<0.8031 -5.0231 0.2456>, rot=<0 0 0 -1>}
 deform_knee_in_l {parent=deform_knee_in_base_l, loc=<0 0 0>, rot=<0 0 0 1>}
 deform_knee_out_base_l {parent=calf_l, loc=<0 4 0>, rot=<0 0 0 -1>}
 deform_knee_out_l {parent=deform_knee_out_base_l, loc=<0 0 0>, rot=<0 0 0 1>}
 foot_pos_corr_l {parent=calf_l, loc=<-43.778 -2.0574 -0.7969>, rot=<0.0427 -0.0238 0.0353 -0.9982>}
 ball_pos_corr_l {parent=foot_pos_corr_l, loc=<-8.4492 9.5049 -0.1128>, rot=<0.0001 -0.0001 -0.7063 -0.7079>}
 thigh_twist_01_l {parent=thigh_l, loc=<-27.952 0 0>, rot=<0.0013 0 0 -1>}
 deform_groin_base_l {parent=thigh_l, loc=<0.4801 5.6776 -0.2486>, rot=<0 0 0 -1>}
 deform_groin_l {parent=deform_groin_base_l, loc=<0 0 0>, rot=<0 0 0 1>}
 thigh_r {parent=pelvis, loc=<-1.0855 -0.2109 8.7053>, rot=<0.0355 -0.0009 0.9993 0.0155>}
 calf_r {parent=thigh_r, loc=<42.5284 0.0019 0.7297>, rot=<0.0006 -0.0161 -0.0391 -0.9991>}
 foot_r {parent=calf_r, loc=<46.7401 2.5014 0.592>, rot=<0.0427 -0.0238 0.0353 -0.9982>}
 ball_r {parent=foot_r, loc=<3.5658 -10.8066 0.1544>, rot=<0.0001 -0.0001 -0.7063 -0.7079>}
 calf_twist_02_r {parent=calf_r, loc=<19.1546 1.7943 0.2638>, rot=<0.0237 0.0095 -0.0035 -0.9997>}
 calf_twist_01_r {parent=calf_r, loc=<35.8483 1.9197 0.5794>, rot=<0.0237 0.0095 -0.0035 -0.9997>}
 foot_corr_bk_r {parent=calf_twist_01_r, loc=<6.8935 4.7046 0.0364>, rot=<-0.0496 -0.727 0.0133 -0.6847>}
 foot_corr_fw_r {parent=calf_twist_01_r, loc=<3.1357 -4.6623 0.374>, rot=<0.7287 -0.0019 0.6824 0.0581>}
 deform_knee_in_base_r {parent=calf_r, loc=<-0.8031 5.0231 -0.2457>, rot=<0 0 0 -1>}
 deform_knee_in_r {parent=deform_knee_in_base_r, loc=<0 0 0>, rot=<0 0 0 1>}
 deform_knee_out_base_r {parent=calf_r, loc=<0 -4 0>, rot=<0 0 0 -1>}
 deform_knee_out_r {parent=deform_knee_out_base_r, loc=<0 0 0>, rot=<0 0 0 1>}
 foot_pos_corr_r {parent=calf_r, loc=<43.778 2.0574 0.7968>, rot=<0.0427 -0.0238 0.0353 -0.9982>}
 ball_pos_corr_r {parent=foot_pos_corr_r, loc=<8.4491 -9.5049 0.1128>, rot=<0.0001 -0.0001 -0.7063 -0.7079>}
 bf_calf_correction {parent=calf_r, loc=<12.664 -0.1362 -3.157>, rot=<-0.6494 0.2426 0.6937 -0.1956>}
 thigh_twist_01_r {parent=thigh_r, loc=<27.952 0 0>, rot=<0.0013 0 0 -1>}
 deform_groin_base_r {parent=thigh_r, loc=<-0.4801 -5.6776 0.2486>, rot=<0 0 0 -1>}
 deform_groin_r {parent=deform_groin_base_r, loc=<0 0 0>, rot=<0 0 0 1>}
 bf_correction {parent=thigh_r, loc=<35.321 4.3372 3.6716>, rot=<-0.6816 0.0015 0.7317 0.0019>}
 deform_glute_base_l {parent=pelvis, loc=<-7.3466 -9.0213 -9.1245>, rot=<0.0172 0.031 -0.4477 -0.8935>}
 deform_glute_l {parent=deform_glute_base_l, loc=<0 0 0>, rot=<0 0 0 1>}
 deform_glute_base_r {parent=pelvis, loc=<-7.3464 -9.0213 9.1245>, rot=<0.031 -0.0172 0.8935 -0.4477>}
 deform_glute_r {parent=deform_glute_base_r, loc=<0 0 0>, rot=<0 0 0 1>}
 dyn_skirt_side_a_base_l {parent=pelvis, loc=<9.0799 0.0741 -12.1592>, rot=<0.1268 -0.7011 -0.6912 -0.1211>}
 dyn_skirt_side_a_l {parent=dyn_skirt_side_a_base_l, loc=<0 0 0>, rot=<0 0 0 1>}
 dyn_skirt_side_b_l {parent=dyn_skirt_side_a_l, loc=<19.9481 0.6491 0.2862>, rot=<0 0 0 1>}
 dyn_skirt_bk_a_base_l {parent=pelvis, loc=<7.0462 -6.4878 -8.0885>, rot=<0.9235 0.1075 0.069 -0.3616>}
 dyn_skirt_bk_a_l {parent=dyn_skirt_bk_a_base_l, loc=<0 0 0>, rot=<0 0 0 1>}
 dyn_skirt_bk_b_l {parent=dyn_skirt_bk_a_l, loc=<-17.8369 0.1048 -1.3002>, rot=<0 0 0 1>}
 dyn_skirt_bk_c_l {parent=dyn_skirt_bk_a_l, loc=<-3.3764 0.1048 -0.2511>, rot=<0 0 0 1>}
 dyn_skirt_bk_a_base_r {parent=pelvis, loc=<6.6858 -6.7724 8.088>, rot=<0.1075 -0.9235 0.3616 0.069>}
 dyn_skirt_bk_a_r {parent=dyn_skirt_bk_a_base_r, loc=<0 0 0>, rot=<0 0 0 1>}
 dyn_skirt_bk_b_r {parent=dyn_skirt_bk_a_r, loc=<17.4647 0.0092 1.0879>, rot=<0 0 0 1>}
 dyn_skirt_bk_c_r {parent=dyn_skirt_bk_a_r, loc=<2.1552 0.0424 0.1647>, rot=<0 0 0 1>}
 dyn_skirt_fr_a_base_r {parent=pelvis, loc=<8.2216 5.404 9.2705>, rot=<-0.0774 -0.9908 -0.0917 0.0631>}
 dyn_skirt_fr_a_r {parent=dyn_skirt_fr_a_base_r, loc=<0 0 0>, rot=<0 0 0 1>}
 dyn_skirt_fr_b_r {parent=dyn_skirt_fr_a_r, loc=<16.6181 0.6401 -0.388>, rot=<0 0 0 1>}
 dyn_skirt_fr_c_r {parent=dyn_skirt_fr_a_r, loc=<2.1162 0.6401 -0.0536>, rot=<0.0211 -0.0188 -0.0092 -0.9996>}
 dyn_skirt_side_a_base_r {parent=pelvis, loc=<7.8767 -0.0598 12.7637>, rot=<0.1235 0.6916 0.7007 -0.1243>}
 dyn_skirt_side_a_r {parent=dyn_skirt_side_a_base_r, loc=<0 0 0>, rot=<0 0 0 1>}
 dyn_skirt_side_b_r {parent=dyn_skirt_side_a_r, loc=<18.6133 -0.8019 0.1925>, rot=<0 0 0 1>}
 dyn_skirt_fr_a_base {parent=pelvis, loc=<8.0261 8.7793 -0.0795>, rot=<-0.0192 -0.9998 0 0.0018>}
 dyn_skirt_fr_a {parent=dyn_skirt_fr_a_base, loc=<0 0 0>, rot=<0 0 0 1>}
 dyn_skirt_fr_b {parent=dyn_skirt_fr_a, loc=<14.8504 0.4014 0.0134>, rot=<0 0 0 1>}
 dyn_skirt_fr_c {parent=dyn_skirt_fr_a, loc=<1.9432 0.6678 0.0181>, rot=<0 0 0 1>}
 dyn_skirt_fr_d {parent=dyn_skirt_fr_c, loc=<2.0741 0.5155 0>, rot=<0 0 0 1>}
 dyn_skirt_bk_a_base {parent=pelvis, loc=<6.4686 -7.9331 0.0859>, rot=<0.094 -0.9956 0.0002 -0.0018>}
 dyn_skirt_bk_a {parent=dyn_skirt_bk_a_base, loc=<0 0 0>, rot=<0 0 0 1>}
 dyn_skirt_bk_b {parent=dyn_skirt_bk_a, loc=<17.34 -1.3241 -0.015>, rot=<0 0 0 1>}
 dyn_skirt_fr_a_base_l {parent=pelvis, loc=<9.3967 5.0321 -9.2476>, rot=<0.9908 -0.0774 0.0631 0.0917>}
 dyn_skirt_fr_a_l {parent=dyn_skirt_fr_a_base_l, loc=<0 0 0>, rot=<0 0 0 1>}
 dyn_skirt_fr_b_l {parent=dyn_skirt_fr_a_l, loc=<-17.0647 -1.6013 0.8671>, rot=<0 0 0 1>}
 dyn_skirt_fr_c_l {parent=dyn_skirt_fr_a_l, loc=<-2.208 -0.1757 0.0554>, rot=<0 0 0 1>}
 skirt_dyn_joints {parent=pelvis, loc=<-97.9752 4.0424 0>, rot=<-0.0127 -0.707 0.0127 -0.707>}
 dyn_skirt_fr_r_base {parent=skirt_dyn_joints, loc=<-8.1397 7.4588 94.7815>, rot=<0 0 0.887 0.4617>}
 dyn_skirt_fr_r_01 {parent=dyn_skirt_fr_r_base, loc=<0 0 0>, rot=<-0.0276 -0.5902 -0.0074 -0.8067>}
 dyn_skirt_fr_r_02 {parent=dyn_skirt_fr_r_01, loc=<3.8946 0 0>, rot=<0.0275 0.1651 0.0158 -0.9858>}
 dyn_skirt_fr_r_03 {parent=dyn_skirt_fr_r_02, loc=<6.1491 0.3503 -2.422>, rot=<0 -0.1305 0 -0.9914>}
 dyn_skirt_fr_r_top {parent=dyn_skirt_fr_r_01, loc=<1.4641 0.1728 1.5215>, rot=<0.0294 0.2999 0.0118 -0.9534>}
 dyn_skirt_fr_l_base {parent=skirt_dyn_joints, loc=<8.7717 8.4145 94.4008>, rot=<0.0553 0.0258 -0.4218 -0.9046>}
 dyn_skirt_fr_l_01 {parent=dyn_skirt_fr_l_base, loc=<0 0 0>, rot=<0.0012 -0.5938 -0.0051 -0.8046>}
 dyn_skirt_fr_l_02 {parent=dyn_skirt_fr_l_01, loc=<4.4582 0 0>, rot=<-0.0168 -0.0308 -0.0027 -0.9994>}
 dyn_skirt_fr_l_03 {parent=dyn_skirt_fr_l_02, loc=<5.3492 0.1712 0.0916>, rot=<0.0558 0.0049 0.0203 -0.9982>}
 dyn_skirt_fr_l_top {parent=dyn_skirt_fr_l_01, loc=<1.0858 0.159 0.6523>, rot=<-0.0005 0.4411 -0.0803 -0.8939>}
 dyn_skirt_side_r_base {parent=skirt_dyn_joints, loc=<-14.3003 0.1125 97.228>, rot=<0 0 1 0>}
 dyn_skirt_side_r_01 {parent=dyn_skirt_side_r_base, loc=<0 0 0>, rot=<0 -0.5548 0 -0.832>}
 dyn_skirt_side_r_02 {parent=dyn_skirt_side_r_01, loc=<4.2479 0 0>, rot=<0.0271 0.0071 -0.0115 -0.9995>}
 dyn_skirt_side_r_03 {parent=dyn_skirt_side_r_02, loc=<6.5935 0.0093 -0.1062>, rot=<0 -0.0408 0 -0.9992>}
 dyn_skirt_side_r_top {parent=dyn_skirt_side_r_01, loc=<1.2272 -0.1409 0.8737>, rot=<0.023 0.278 -0.0185 -0.9601>}
 dyn_skirt_side_l_base {parent=skirt_dyn_joints, loc=<14.9392 -0.1617 98.8221>, rot=<0 0 0 1>}
 dyn_skirt_side_l_01 {parent=dyn_skirt_side_l_base, loc=<0 0 0>, rot=<-0.0309 -0.5458 0.0473 -0.836>}
 dyn_skirt_side_l_02 {parent=dyn_skirt_side_l_01, loc=<4.1712 0 0>, rot=<0.0695 -0.0342 0.0166 -0.9969>}
 dyn_skirt_side_l_03 {parent=dyn_skirt_side_l_02, loc=<4.8926 0.1828 -0.0524>, rot=<0.0005 -0.0038 -0.0435 -0.999>}
 dyn_skirt_bck_r_base {parent=skirt_dyn_joints, loc=<-11.4044 -9.2478 98.1369>, rot=<0 0 0.9397 -0.342>}
 dyn_skirt_bck_r_01 {parent=dyn_skirt_bck_r_base, loc=<0 0 0>, rot=<-0.0263 -0.5604 0.0388 -0.8269>}
 dyn_skirt_bck_r_02 {parent=dyn_skirt_bck_r_01, loc=<3.9702 0 0>, rot=<0.1177 0.1248 -0.0652 -0.983>}
 dyn_skirt_bck_r_03 {parent=dyn_skirt_bck_r_02, loc=<5.9573 -0.1932 -3.013>, rot=<0 -0.1305 0 -0.9914>}
 dyn_skirt_bck_r_top {parent=dyn_skirt_bck_r_01, loc=<0.9161 -0.0473 1.0341>, rot=<0.1075 0.2596 -0.0808 -0.9563>}
 dyn_skirt_bck_c_base {parent=skirt_dyn_joints, loc=<0 -12.2859 98.0305>, rot=<0 0 0.7071 -0.7071>}
 dyn_skirt_bck_c_01 {parent=dyn_skirt_bck_c_base, loc=<0 0 0>, rot=<0 -0.5609 0 -0.8279>}
 dyn_skirt_bck_c_02 {parent=dyn_skirt_bck_c_01, loc=<4.2167 0 0>, rot=<0 -0.0752 0 -0.9972>}
 dyn_skirt_bck_c_03 {parent=dyn_skirt_bck_c_02, loc=<7.185 0 0>, rot=<0 0 0 1>}
 dyn_skirt_bck_c_top {parent=dyn_skirt_bck_c_01, loc=<4.6166 0 1.4871>, rot=<0 -0.0437 0 -0.999>}
 dyn_skirt_bck_l_base {parent=skirt_dyn_joints, loc=<10.6147 -8.5558 99.0531>, rot=<0 0 0.4226 -0.9063>}
 dyn_skirt_bck_l_01 {parent=dyn_skirt_bck_l_base, loc=<0 0 0>, rot=<-0.0017 -0.5488 0.0026 -0.836>}
 dyn_skirt_bck_l_02 {parent=dyn_skirt_bck_l_01, loc=<4.1313 0 0>, rot=<0.0693 0.1372 -0.0422 -0.9872>}
 dyn_skirt_bck_l_03 {parent=dyn_skirt_bck_l_02, loc=<5.5404 -0.3211 -1.803>, rot=<0 -0.1305 0 -0.9914>}
 ik_hand_root {parent=root, loc=<0 0 0>, rot=<0 0 0 1>}
 ik_hand_gun {parent=ik_hand_root, loc=<-48.1743 3.3545 104.4381>, rot=<0.6276 0.1309 0.4183 -0.6434>}
 ik_foot_root {parent=root, loc=<0 0 0>, rot=<0 0 0 1>}
 ik_foot_r {parent=ik_foot_root, loc=<-14.8262 -7.3545 8.296>, rot=<0.7082 0.0284 -0.7045 0.0357>}
]
