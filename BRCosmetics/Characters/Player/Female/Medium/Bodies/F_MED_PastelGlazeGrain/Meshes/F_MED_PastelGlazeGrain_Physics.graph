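# Physics asset for the PastelGlazeGrain medium female player body, defining collision and joints for its dynamic skirt bones.
# Capsules on pelvis, spine_01, both thighs and right calf cover the body alongside front, back and side skirt shapes.
# Tiny base bodies join the pelvis with free rotation, anchoring branching skirt chains with tight swing and twist limits.
# The right side is asymmetric, adding sphere bodies and a free front-center-to-right link.
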PhysicsAsset /BRCosmetics/Characters/Player/Female/Medium/Bodies/F_MED_PastelGlazeGrain/Meshes/F_MED_PastelGlazeGrain_Physics
bodies {
 calf_r [{
   shapeType=Capsule
   length=32.2975
   radius=11.5856
   shapeName=calf_r_capsule
   center=<39.4101 2.3445 -0.399>
   rotation=<rot 90 80.7735 100.4081>
  }]
 dyn_skirt_bk_a_base [{shapeType=Capsule, length=0.505, radius=0.505, shapeName=dyn_skirt_bk_a_base_capsule, rotation=<rot 0 -90 0>}]
 dyn_skirt_bk_a_base_l [{shapeType=Capsule, length=0.505, radius=0.505, shapeName=dyn_skirt_bk_a_base_l_capsule, rotation=<rot 0 -90 0>}]
 dyn_skirt_bk_a_base_r [{shapeType=Capsule, length=0.505, radius=0.505, shapeName=dyn_skirt_bk_a_base_r_capsule, rotation=<rot 0 -90 0>}]
 dyn_skirt_bk_a_l [{
   shapeType=Capsule
   length=6
   radius=7
   shapeName=dyn_skirt_bk_a_l_capsule
   center=<-9.3497 2.1048 -5.5486>
   rotation=<rot 8.5506 6.7559 38.5>
  }]
 dyn_skirt_bk_a_r [{
   shapeType=Capsule
   length=6
   radius=7
   shapeName=dyn_skirt_bk_a_r_capsule
   center=<9.4879 -3.0291 4.0132>
   rotation=<rot 8.7062 5.3106 -148.4528>
  }]
 dyn_skirt_bk_b_l [{
   shapeType=Capsule
   length=6
   radius=7
   shapeName=dyn_skirt_bk_b_l_capsule
   center=<-2.2856 0.5867 -5.3696>
   rotation=<rot 8.5506 6.7559 38.5>
  }]
 dyn_skirt_bk_b_r [{
   shapeType=Capsule
   length=6
   radius=7
   shapeName=dyn_skirt_bk_b_r_capsule
   center=<3.2233 -0.7645 3.9613>
   rotation=<rot 4.2589 12.2033 -146.8088>
  }]
 dyn_skirt_bk_c_l [{shapeType=Capsule, length=2, radius=3, shapeName=dyn_skirt_bk_c_l_capsule, center=<-10.958 1.6254 -1.1895>}]
 dyn_skirt_bk_c_r [{shapeType=Capsule, length=2, radius=3, shapeName=dyn_skirt_bk_c_r_capsule, center=<10.5785 -4.5729 3.7185>}]
 dyn_skirt_fr_a [{shapeType=Capsule, length=11.3107, radius=5.4135, shapeName=dyn_skirt_fr_a_capsule, center=<8.7217 0.3875 0.8155>}]
 dyn_skirt_fr_a_base [{shapeType=Capsule, length=0.505, radius=0.6653, shapeName=dyn_skirt_fr_a_base_capsule, rotation=<rot 0 -90 0>}]
 dyn_skirt_fr_a_base_l [{shapeType=Capsule, length=0.2355, radius=1.0942, shapeName=dyn_skirt_fr_a_base_l_capsule, rotation=<rot 0 -90 0>}]
 dyn_skirt_fr_a_base_r [{shapeType=Capsule, length=0.505, radius=0.505, shapeName=dyn_skirt_fr_a_base_r_capsule, rotation=<rot 0 -90 0>}]
 dyn_skirt_fr_a_l [{
   shapeType=Capsule
   length=6
   radius=7
   shapeName=dyn_skirt_fr_a_l_capsule
   center=<-10.947 -4.0384 -1.3112>
   rotation=<rot 0.0001 0.0003 -4.1594>
  }]
 dyn_skirt_fr_a_r [{
   shapeType=Capsule
   length=6
   radius=7
   shapeName=dyn_skirt_fr_a_r_capsule
   center=<10.9469 3.9433 2.5759>
   rotation=<rot 4.1026 0.0007 -179.9999>
  }]
 dyn_skirt_fr_b [{shapeType=Capsule, length=8.355, radius=5.6056, shapeName=dyn_skirt_fr_b_capsule, center=<8.2946 0.0304 0.4243>}]
 dyn_skirt_fr_b_l [{
   shapeType=Capsule
   length=6
   radius=7
   shapeName=dyn_skirt_fr_b_l_capsule
   center=<-5.1356 -2.2723 -2.42>
   rotation=<rot 0.0001 0.0003 -4.1594>
  }]
 dyn_skirt_fr_b_r [{
   shapeType=Capsule
   length=6
   radius=7
   shapeName=dyn_skirt_fr_b_r_capsule
   center=<5.082 3.1608 3.6033>
   rotation=<rot 4.1026 0.0007 -179.9999>
  }]
 dyn_skirt_fr_c [{shapeType=Capsule, length=2, radius=3, shapeName=dyn_skirt_fr_c_capsule, center=<8.3904 0 0>}]
 dyn_skirt_fr_c_l [{shapeType=Capsule, length=2, radius=3, shapeName=dyn_skirt_fr_c_l_capsule, center=<-15.3384 -4.923 2.1229>}]
 dyn_skirt_fr_c_r [{shapeType=Capsule, length=2, radius=3, shapeName=dyn_skirt_fr_c_r_capsule, center=<3.9682 0.6881 0.4759>}]
 dyn_skirt_fr_d [{shapeType=Capsule, length=5.7819, radius=4.3232, shapeName=dyn_skirt_fr_d_capsule, center=<4.4223 0 0>}]
 dyn_skirt_side_a_base_l [{shapeType=Capsule, length=0.505, radius=0.505, shapeName=dyn_skirt_side_a_base_l_capsule, rotation=<rot 0 -90 0>}]
 dyn_skirt_side_a_base_r [{shapeType=Capsule, length=0.505, radius=0.505, shapeName=dyn_skirt_side_a_base_r_capsule, rotation=<rot 0 -90 0>}]
 dyn_skirt_side_a_l [{shapeType=Capsule, length=6, radius=7, shapeName=dyn_skirt_side_a_l_capsule, center=<11.7757 -4.9437 0>, rotation=<rot 0 5 0>}]
 dyn_skirt_side_a_r [{
   shapeType=Capsule
   length=6
   radius=7
   shapeName=dyn_skirt_side_a_r_capsule
   center=<10.791 5.8644 1.2224>
   rotation=<rot 0.0004 -7.4477 -179.9999>
  }]
 dyn_skirt_side_b_l [{shapeType=Capsule, length=6, radius=7, shapeName=dyn_skirt_side_b_l_capsule, center=<3.3096 -3.5325 0>, rotation=<rot 0 10.6725 0>}]
 dyn_skirt_side_b_r [{
   shapeType=Capsule
   length=6
   radius=7
   shapeName=dyn_skirt_side_b_r_capsule
   center=<3.369 4.1767 1.3653>
   rotation=<rot 0.0008 -12.6082 179.9997>
  }]
 dyn_skirt_side_c_r [{shapeType=Sphere, radius=3.4007, shapeName=dyn_skirt_side_c_r_sphere, center=<10.6387 0 0>}]
 dyn_skirt_side_d_r [{shapeType=Sphere, radius=2.5722, shapeName=dyn_skirt_side_d_r_sphere, center=<5.5789 0 0>}]
 dyn_skirt_side_d_r_base [{shapeType=Sphere, radius=0.505, shapeName=dyn_skirt_side_d_r_base_sphere}]
 pelvis [{shapeType=Capsule, length=15.2993, radius=7.5848, center=<-5.2316 -1.146 0>, rotation=<rot 0 -90 0>}]
 spine_01 [{shapeType=Capsule, length=20.4794, radius=7.7191, shapeName=spine_01_capsule, center=<-2.5766 -3.1453 0>}]
 thigh_l [{
   shapeType=Capsule
   length=26.7313
   radius=8.2686
   shapeName=thigh_l_capsule
   center=<-16.8998 0.6736 -0.2476>
   rotation=<rot 3.7075 -90.8125 -95.5759>
  }]
 thigh_r [{
   shapeType=Capsule
   length=29.4895
   radius=8.2686
   shapeName=thigh_l_capsule
   center=<16.9352 -0.1632 0.8274>
   rotation=<rot -0.1512 -91.2545 88.8615>
  }]
}
constraints [
 {
  bone1Name=dyn_skirt_bk_a_base
  bone2Name=pelvis
  swing1LimitDegrees=45
  swing1Motion=Free
  swing2LimitDegrees=45
  swing2Motion=Free
  twistLimitDegrees=45
  twistMotion=Free
 }
 {
  bone1Name=dyn_skirt_bk_a_base_l
  bone2Name=pelvis
  swing1LimitDegrees=45
  swing1Motion=Free
  swing2LimitDegrees=45
  swing2Motion=Free
  twistLimitDegrees=45
  twistMotion=Free
 }
 {
  bone1Name=dyn_skirt_bk_a_base_r
  bone2Name=pelvis
  swing1LimitDegrees=45
  swing1Motion=Free
  swing2LimitDegrees=45
  swing2Motion=Free
  twistLimitDegrees=45
  twistMotion=Free
 }
 {
  bone1Name=dyn_skirt_fr_a_base
  bone2Name=pelvis
  swing1LimitDegrees=45
  swing1Motion=Free
  swing2LimitDegrees=45
  swing2Motion=Free
  twistLimitDegrees=45
  twistMotion=Free
 }
 {
  bone1Name=dyn_skirt_fr_a_base_l
  bone2Name=pelvis
  swing1LimitDegrees=45
  swing1Motion=Free
  swing2LimitDegrees=45
  swing2Motion=Free
  twistLimitDegrees=45
  twistMotion=Free
 }
 {
  bone1Name=dyn_skirt_fr_a_base_r
  bone2Name=pelvis
  swing1LimitDegrees=45
  swing1Motion=Free
  swing2LimitDegrees=45
  swing2Motion=Free
  twistLimitDegrees=45
  twistMotion=Free
 }
 {
  bone1Name=dyn_skirt_side_a_base_l
  bone2Name=pelvis
  swing1LimitDegrees=45
  swing1Motion=Free
  swing2LimitDegrees=45
  swing2Motion=Free
  twistLimitDegrees=45
  twistMotion=Free
 }
 {
  bone1Name=dyn_skirt_side_a_base_r
  bone2Name=pelvis
  swing1LimitDegrees=45
  swing1Motion=Free
  swing2LimitDegrees=45
  swing2Motion=Free
  twistLimitDegrees=45
  twistMotion=Free
 }
 {
  bone1Name=thigh_l
  bone2Name=pelvis
  swing1LimitDegrees=45
  swing1Motion=Free
  swing2LimitDegrees=45
  swing2Motion=Free
  twistLimitDegrees=45
  twistMotion=Free
 }
 {
  bone1Name=thigh_r
  bone2Name=pelvis
  swing1LimitDegrees=45
  swing1Motion=Free
  swing2LimitDegrees=45
  swing2Motion=Free
  twistLimitDegrees=45
  twistMotion=Free
 }
 {bone1Name=dyn_skirt_fr_a_l, bone2Name=dyn_skirt_fr_a_base_l, swing1LimitDegrees=20, swing2LimitDegrees=5, twistLimitDegrees=10}
 {bone1Name=dyn_skirt_fr_b_l, bone2Name=dyn_skirt_fr_a_l, swing1LimitDegrees=20, swing2LimitDegrees=5, twistLimitDegrees=10}
 {bone1Name=dyn_skirt_fr_b_r, bone2Name=dyn_skirt_fr_a_r, swing1LimitDegrees=20, swing2LimitDegrees=5, twistLimitDegrees=10}
 {bone1Name=dyn_skirt_fr_a_r, bone2Name=dyn_skirt_fr_a_base_r, swing1LimitDegrees=30, swing2LimitDegrees=5, twistLimitDegrees=10}
 {bone1Name=dyn_skirt_side_a_l, bone2Name=dyn_skirt_side_a_base_l, swing1LimitDegrees=20, swing2LimitDegrees=3, twistLimitDegrees=5}
 {bone1Name=dyn_skirt_side_b_l, bone2Name=dyn_skirt_side_a_l, swing1LimitDegrees=20, swing2LimitDegrees=10, twistLimitDegrees=5}
 {bone1Name=dyn_skirt_side_b_r, bone2Name=dyn_skirt_side_a_r, swing1LimitDegrees=20, swing2LimitDegrees=20, twistLimitDegrees=5}
 {bone1Name=dyn_skirt_side_a_r, bone2Name=dyn_skirt_side_a_base_r, swing1LimitDegrees=20, swing2LimitDegrees=10, twistLimitDegrees=5}
 {bone1Name=dyn_skirt_bk_a_r, bone2Name=dyn_skirt_bk_a_base_r, swing1LimitDegrees=10, swing2LimitDegrees=20, twistLimitDegrees=5}
 {bone1Name=dyn_skirt_bk_b_r, bone2Name=dyn_skirt_bk_a_r, swing1LimitDegrees=50, swing2LimitDegrees=10, twistLimitDegrees=5}
 {bone1Name=dyn_skirt_bk_a_l, bone2Name=dyn_skirt_bk_a_base_l, swing1LimitDegrees=50, swing2LimitDegrees=10, twistLimitDegrees=5}
 {bone1Name=dyn_skirt_bk_b_l, bone2Name=dyn_skirt_bk_a_l, swing1LimitDegrees=50, swing2LimitDegrees=10, twistLimitDegrees=5}
 {bone1Name=dyn_skirt_fr_b, bone2Name=dyn_skirt_fr_a, swing1LimitDegrees=10, swing2LimitDegrees=5, twistLimitDegrees=10}
 {bone1Name=dyn_skirt_fr_a, bone2Name=dyn_skirt_fr_a_base, swing1LimitDegrees=10, swing2LimitDegrees=5, twistLimitDegrees=10}
 {bone1Name=dyn_skirt_bk_c_l, bone2Name=dyn_skirt_bk_a_l, swing1LimitDegrees=20, swing2LimitDegrees=10, twistLimitDegrees=2}
 {bone1Name=dyn_skirt_bk_c_r, bone2Name=dyn_skirt_bk_a_r, swing1LimitDegrees=10, swing2LimitDegrees=10, twistLimitDegrees=2}
 {bone1Name=dyn_skirt_fr_c, bone2Name=dyn_skirt_fr_a, swing1LimitDegrees=5, swing2LimitDegrees=10, twistLimitDegrees=2}
 {bone1Name=dyn_skirt_fr_c_l, bone2Name=dyn_skirt_fr_a_l, swing1LimitDegrees=10, swing2LimitDegrees=10, twistLimitDegrees=2}
 {bone1Name=dyn_skirt_fr_c_r, bone2Name=dyn_skirt_fr_a_r, swing1LimitDegrees=20, swing2LimitDegrees=10, twistLimitDegrees=2}
 {bone1Name=dyn_skirt_fr_d, bone2Name=dyn_skirt_fr_c, swing1LimitDegrees=5, swing2LimitDegrees=10, twistLimitDegrees=7}
 {
  bone1Name=dyn_skirt_fr_b_r
  bone2Name=dyn_skirt_fr_b
  swing1LimitDegrees=45
  swing1Motion=Free
  swing2LimitDegrees=45
  swing2Motion=Free
  twistLimitDegrees=45
  twistMotion=Free
 }
 {
  bone1Name=calf_r
  bone2Name=thigh_r
  swing1LimitDegrees=45
  swing1Motion=Free
  swing2LimitDegrees=45
  swing2Motion=Free
  twistLimitDegrees=45
  twistMotion=Free
 }
 {bone1Name=dyn_skirt_side_c_r, bone2Name=dyn_skirt_side_a_r, swing1LimitDegrees=10, swing2LimitDegrees=15, twistLimitDegrees=2}
 {bone1Name=dyn_skirt_side_d_r, bone2Name=dyn_skirt_side_d_r_base, swing1LimitDegrees=15, swing2LimitDegrees=15, twistLimitDegrees=3}
 {
  bone1Name=dyn_skirt_side_d_r_base
  bone2Name=pelvis
  swing1LimitDegrees=45
  swing1Motion=Free
  swing2LimitDegrees=45
  swing2Motion=Free
  twistLimitDegrees=45
  twistMotion=Free
 }
]
