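# Animation Blueprint for the PastelGlazeGrain female body, adding procedural skirt motion to the pose copied from its parent mesh.
# A Control Rig feeds pelvis-space RigidBody physics, then AnimDynamics springs on two right-side skirt bones and constraints on in-between bones.
# An additive rotation adjusts the front-left skirt bone in the front end.
# Each update sets RigidBody gravity through a deprecated override function and lowers its alpha while skydiving.
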AnimBlueprint /BRCosmetics/Characters/Player/Female/Medium/Bodies/F_MED_PastelGlazeGrain/Meshes/F_MED_PastelGlazeGrain_AnimBP
target_skeleton=&/F_MED_PastelGlazeGrain_Skeleton
graphs [
 AnimGraph {nodes [
   n1 = Control Rig(Source=Copy Pose From Mesh() {bUseAttachedParent=true}) {
    alphaBoolBlend {blendOption=Linear}
    controlRigAssetReference {blueprintRigClass=&/F_MED_PastelGlazeGrain_CtrlRig.F_MED_PastelGlazeGrain_CtrlRig_C}
   }
   n2 = RigidBody!(
    ComponentPose=Local To Component(LocalPose=@n1)
    Alpha=Get Dyn Skirt() -> (Dyn Skirt:double)
    OverrideWorldGravity:Vector=Get Gravity() -> (Gravity:Vector)
   ) -> (Pose) {
    bOverrideWorldGravity=true
    overridePhysicsAsset=&/F_MED_PastelGlazeGrain_Physics
    simulationSpace=BaseBoneSpace
    alphaBoolBlend {blendInTime=0.5, blendOption=Linear, blendOutTime=0.5}
    baseBoneRef {boneName=pelvis}
    componentAppliedLinearAccClamp=<5000 5000 5000>
    componentLinearAccScale=<0.4 0.5 0.5>
    componentLinearVelScale=<0.3 0.3 0.3>
    overrideWorldGravity {z=-980}
   }
   n3 = Anim Dynamics - Bone: dyn_skirt_bk_c_r(ComponentPose=@n2, Alpha=1, ExternalForce="0,0,0") {
    angularSpringConstant=500
    bAngularSpring=true
    bOverrideLinearDamping=true
    linearDampingOverride=0.3
    simulationSpace=RootRelative
    alphaBoolBlend {blendOption=Linear}
    boundBone {boneName=dyn_skirt_bk_c_r}
    physicsBodyDefinitions [{
      boundBone {boneName=dyn_skirt_bk_c_r}
      boxExtents=<10 10 10>
      constraintSetup {angularLimitsMax=<0 5 5>, angularLimitsMin=<0 -5 -5>, angularTarget=<0 0 1>}
      localJointOffset=<0.1 0 0>
     }]
   }
   n4 = Anim Dynamics - Bone: dyn_skirt_side_d_r(ComponentPose=@n3, Alpha=1, ExternalForce="0,0,0") {
    angularSpringConstant=500
    bAngularSpring=true
    bOverrideLinearDamping=true
    linearDampingOverride=0.3
    simulationSpace=RootRelative
    alphaBoolBlend {blendOption=Linear}
    boundBone {boneName=dyn_skirt_side_d_r}
    physicsBodyDefinitions [{
      boundBone {boneName=dyn_skirt_side_d_r}
      boxExtents=<10 10 10>
      constraintSetup {angularLimitsMax=<0 5 5>, angularLimitsMin=<0 -5 -5>, angularTarget=<0 0 1>}
      localJointOffset=<0.1 0 0>
     }]
   }
   n5 = "Constraint - dyn_skirt_side_d_r_base (PPP)"!(ComponentPose=@n4, Alpha=1, ConstraintWeights_0=0.3, ConstraintWeights_1:float=0.3, ConstraintWeights_2:float=0.4) -> (Pose) {
    alphaBoolBlend {blendOption=Linear}
    boneToModify {boneName=dyn_skirt_side_d_r_base}
    constraintSetup [{targetBone {boneName=dyn_skirt_side_a_base_r}}, {targetBone {boneName=dyn_skirt_fr_a_base_r}}, {targetBone {boneName=pelvis}}]
    constraintWeights [1,1,1]
   }
   n6 = "Constraint - dyn_skirt_bk_a_base (PP)"!(ComponentPose=@n5, Alpha=1, ConstraintWeights_0=0.5, ConstraintWeights_1:float=0.5) -> (Pose) {
    alphaBoolBlend {blendOption=Linear}
    boneToModify {boneName=dyn_skirt_bk_a_base}
    constraintSetup [{targetBone {boneName=dyn_skirt_bk_a_l}}, {targetBone {boneName=dyn_skirt_bk_a_r}}]
    constraintWeights [1,1]
   }
   n7 = "Constraint - dyn_skirt_bk_b (PP)"!(ComponentPose=@n6, Alpha=1, ConstraintWeights_0=0.5, ConstraintWeights_1:float=0.5) -> (Pose) {
    alphaBoolBlend {blendOption=Linear}
    boneToModify {boneName=dyn_skirt_bk_b}
    constraintSetup [{targetBone {boneName=dyn_skirt_bk_b_l}}, {targetBone {boneName=dyn_skirt_bk_b_r}}]
    constraintWeights [1,1]
   }
   n8 = "Transform (Modify) Bone - Bone: dyn_skirt_fr_a_l"!(ComponentPose=@n7, bAlphaBoolEnabled:bool=Get bIsInFrontEnd(), Translation="0,0,0", Rotation="-26.456,0.069,-0.129", Scale="1,1,1") -> (Pose) {
    alphaInputType=Bool
    rotationMode=BMM_Additive
    rotationSpace=BCS_BoneSpace
    alphaBoolBlend {blendOption=Linear}
    boneToModify {boneName=dyn_skirt_fr_a_l}
    rotation=<rot -26.456 0.069 -0.129>
   }
   Output Pose(Result=Component To Local(ComponentPose=@n8)) {layerGroup=None}
  ]}
 EventGraph {nodes [
   Event BlueprintUpdateAnimation()
   > n1 = Sequence()
   n2 = "To Float (Boolean)"(InBool=Get bIsSkyDiving())
   Set Dyn Skirt(
    execute=@n1.then_1
    Dyn Skirt:double=MapRangeClamped(Value=0, InRangeA=0, InRangeB=1, OutRangeA=MapRangeClamped(Value=@n2, InRangeA=0, InRangeB=1, OutRangeA=1, OutRangeB=0.3), OutRangeB=0.5)
   ) -> (Output_Get:double)
   n3 = Gravity Override State Function DEPRECATED(
    execute=@n1.then_0
    CharacterPartAnimInstance=Self-Reference()
    GameTimeInSeconds=GetGameTimeInSeconds()
    GravityOverrideParams=Get Gravity Override Params()
   )
   > Set Gravity(Gravity:Vector=@n3) -> (Output_Get:Vector)
  ]}
]
variables [
 Dyn Skirt {type=number}
 Gravity {type=Vector}
 Gravity Override Params {type=GravityOverrideParamsStruct}
]
