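# Control Rig for the face accessory part of the F_MED_PastelGlazeGrain female character, procedurally posing its back, left and right dynamic hair chains.
# A first pass parent-constrains dyn_hair_bk_03 (twice, identically) and dyn_hair_r_04 to spine_05.
# Spherical pose readers on head and both upper arms then weight rotation offsets on the hair master bones.
# When Is skydive is set, an additional full-weight rotation is applied to the back hair master.
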ControlRigBlueprint /BRCosmetics/Characters/Player/Female/Medium/Bodies/F_MED_PastelGlazeGrain/Meshes/Parts/F_MED_PastelGlazeGrain_FaceAcc_CtrlRig
graphs [
 RigVMFunctionLibrary {links [], nodes []}
 RigVMModel {nodes [
   RigUnit_BeginExecution = &RigUnit_BeginExecution()
   RigVMFunction_Sequence = &RigVMFunction_Sequence(ExecuteContext=@RigUnit_BeginExecution.ExecutePin)
   &RigUnit_ParentConstraint(
    AdvancedSettings=(InterpolationType=Average,RotationOrderForFilter=XZY)
    Child=(Type=Bone,Name="dyn_hair_bk_03")
    Parents=((Item=(Type=Bone,Name="spine_05"),Weight=1))
    Weight=1
    ExecutePin=@RigVMFunction_Sequence.A
   )
   > &RigUnit_ParentConstraint(
    AdvancedSettings=(InterpolationType=Average,RotationOrderForFilter=XZY)
    Child=(Type=Bone,Name="dyn_hair_bk_03")
    Parents=((Item=(Type=Bone,Name="spine_05"),Weight=1))
    Weight=1
   )
   > &RigUnit_ParentConstraint(
    AdvancedSettings=(InterpolationType=Average,RotationOrderForFilter=XZY)
    Child=(Type=Bone,Name="dyn_hair_r_04")
    Parents=((Item=(Type=Bone,Name="spine_05"),Weight=1))
    Weight=1
   )
   n1 = &RigUnit_SphericalPoseReader(
    ActiveRegionScaleFactors=(PositiveWidth=1,NegativeWidth=1,PositiveHeight=1,NegativeHeight=1)
    ActiveRegionSize=0.1
    Debug=(bDrawDebug=True,bDraw2D=False,bDrawLocalAxes=False,DebugScale=25,DebugSegments=20,DebugThickness=0.25)
    DriverAxis=(X=1,Y=0,Z=0)
    DriverItem=(Type=Bone,Name="head")
    FalloffRegionScaleFactors=(PositiveWidth=1,NegativeWidth=1,PositiveHeight=1,NegativeHeight=1)
    FalloffSize=0.2
    FlipHeightScaling=False
    FlipWidthScaling=False
    OptionalParentItem=(Type=Bone,Name="None")
    RotationOffset=(X=90,Y=0,Z=-140)
    ExecutePin=@RigVMFunction_Sequence.B
   )
   > &RigUnit_ModifyTransforms(
    ItemToModify="((Item=(Type=Bone,Name=\"dyn_hair_bk_MSTR\"),Transform=(Rotation=(X=0,Y=-0.1736,Z=0,W=0.9848),Translation=(X=0,Y=0,Z=0),Scale3D=(X=1,Y=1,Z=1))),(Item=(Type=Bone,Name=\"dyn_hair_l_MSTR\"),Transform=(Rotation=(X=0.0449,Y=-0.2549,Z=0.1677,W=0.9513),Translation=(X ...(truncated)"
    WeightMinimum=0
    Weight=@n1.OutputParam
   )
   > n2 = &RigUnit_SphericalPoseReader(
    ActiveRegionScaleFactors=(PositiveWidth=1,NegativeWidth=1,PositiveHeight=2,NegativeHeight=2)
    ActiveRegionSize=0.05
    Debug=(bDrawDebug=true,bDraw2D=False,bDrawLocalAxes=False,DebugScale=25,DebugSegments=20,DebugThickness=0.25)
    DriverAxis=(X=1,Y=0,Z=0)
    DriverItem=(Type=Bone,Name="upperarm_l")
    FalloffRegionScaleFactors=(PositiveWidth=1.2,NegativeWidth=1,PositiveHeight=1.5,NegativeHeight=1.5)
    FalloffSize=0.25
    FlipHeightScaling=False
    FlipWidthScaling=False
    OptionalParentItem=(Type=Bone,Name="None")
    RotationOffset=(X=0,Y=-55,Z=0)
   )
   > &RigUnit_ModifyTransforms(
    ItemToModify=((Item=(Type=Bone,Name="dyn_hair_l_MSTR"),Transform=(Rotation=(X=0.0298,Y=-0.2938,Z=-0.0965,W=0.9505),Translation=(X=0,Y=0,Z=0),Scale3D=(X=1,Y=1,Z=1))))
    WeightMinimum=0
    Weight=@n2.OutputParam
   )
   > n3 = &RigUnit_SphericalPoseReader(
    ActiveRegionScaleFactors=(PositiveWidth=1,NegativeWidth=1,PositiveHeight=2,NegativeHeight=2)
    ActiveRegionSize=0.05
    Debug=(bDrawDebug=true,bDraw2D=False,bDrawLocalAxes=False,DebugScale=25,DebugSegments=20,DebugThickness=0.25)
    DriverAxis=(X=-1,Y=0,Z=0)
    DriverItem=(Type=Bone,Name="upperarm_r")
    FalloffRegionScaleFactors=(PositiveWidth=1.2,NegativeWidth=1,PositiveHeight=1.5,NegativeHeight=1.5)
    FalloffSize=0.25
    FlipHeightScaling=False
    FlipWidthScaling=False
    OptionalParentItem=(Type=Bone,Name="None")
    RotationOffset=(X=0,Y=125,Z=0)
   )
   > n4 = &RigUnit_ModifyTransforms(
    ItemToModify=((Item=(Type=Bone,Name="dyn_hair_r_MSTR"),Transform=(Rotation=(X=0,Y=-0.2954,Z=0,W=0.9554),Translation=(X=0,Y=0,Z=0),Scale3D=(X=1,Y=1,Z=1))))
    WeightMinimum=0
    Weight=@n3.OutputParam
   )
   RigVMFunction_ControlFlowBranch = &RigVMFunction_ControlFlowBranch(ExecuteContext=@n4.ExecutePin, Condition=RigVMVariableNode() {variable=Is skydive}.Value)
   &RigUnit_ModifyTransforms(
    ItemToModify=((Item=(Type=Bone,Name="dyn_hair_bk_MSTR"),Transform=(Rotation=(X=0,Y=0.1736,Z=0,W=0.9848),Translation=(X=0,Y=0,Z=0),Scale3D=(X=1,Y=1,Z=1))))
    Weight=1
    WeightMinimum=0
    ExecutePin=@RigVMFunction_ControlFlowBranch.True
   )
  ]}
]
hierarchy [
 root {loc=<0 0 0>, rot=<0 0 0 1>}
 pelvis {parent=root, loc=<0 -0.5301 98.0571>, rot=<0.0127 0.707 -0.0127 -0.707>}
 spine_01 {parent=pelvis, loc=<6.1208 -0.9328 0>, rot=<0 0 0.034 -0.9994>}
 spine_02 {parent=spine_01, loc=<6.5996 -0.2095 0>, rot=<0 0 0.0238 -0.9997>}
 spine_03 {parent=spine_02, loc=<6.4914 -0.1237 0>, rot=<0 0 0.0149 -0.9999>}
 spine_04 {parent=spine_03, loc=<6.9598 0.0794 0>, rot=<0 0 -0.0132 -0.9999>}
 spine_05 {parent=spine_04, loc=<6.9987 0 0>, rot=<0 0 -0.0015 -1>}
 attach_cape {parent=spine_05, loc=<9.017 -11.714 0>, rot=<0.0616 -0.7044 -0.0616 -0.7044>}
 neck_01 {parent=spine_05, loc=<11.3138 0.1529 0>, rot=<0 0 -0.2003 -0.9797>}
 neck_02 {parent=neck_01, loc=<4.5627 0 0>, rot=<0 0 0.0504 -0.9987>}
 head {parent=neck_02, loc=<5.3914 0 0>, rot=<0 0 0.115 -0.9934>}
 dyn_hair_bk_MSTR {parent=head, loc=<11.6835 -5.2363 0>, rot=<0.1215 -0.6966 0.6966 -0.1215>}
 dyn_hair_bk_00 {parent=dyn_hair_bk_MSTR, loc=<0.5 0 0>, rot=<0 -0.0572 0 -0.9984>}
 dyn_hair_bk_01 {parent=dyn_hair_bk_00, loc=<6.0938 0 0>, rot=<0 -0.0272 0 -0.9996>}
 dyn_hair_bk_02 {parent=dyn_hair_bk_01, loc=<8.7854 0 0>, rot=<0 0.1593 0 -0.9872>}
 dyn_hair_bk_03 {parent=dyn_hair_bk_02, loc=<9.4823 0 0>, rot=<0 -0.1683 0 -0.9857>}
 dyn_hair_bk_04 {parent=dyn_hair_bk_03, loc=<10.3536 0 0>, rot=<0 -0.0751 0 -0.9972>}
 dyn_hair_bk_05 {parent=dyn_hair_bk_04, loc=<8.7412 0 0>, rot=<0 0 0 1>}
 dyn_hair_l_MSTR {parent=head, loc=<8.8541 0.6916 -6.499>, rot=<0.0262 -0.7919 0.5984 -0.1187>}
 dyn_hair_l_00 {parent=dyn_hair_l_MSTR, loc=<0.5 0 0>, rot=<0 0 0 1>}
 dyn_hair_l_01 {parent=dyn_hair_l_00, loc=<7.0495 0 0>, rot=<0.0202 0.1166 0.0179 -0.9928>}
 dyn_hair_l_02 {parent=dyn_hair_l_01, loc=<7.1616 0 0>, rot=<-0.086 0.2331 0.0422 -0.9677>}
 dyn_hair_l_03 {parent=dyn_hair_l_02, loc=<8.7592 0 0>, rot=<-0.0205 -0.2538 0.0767 -0.964>}
 dyn_hair_l_04 {parent=dyn_hair_l_03, loc=<8.993 0 0>, rot=<0.0802 -0.1932 -0.017 -0.9777>}
 dyn_hair_l_05 {parent=dyn_hair_l_04, loc=<11.6454 0 0>, rot=<0 0 0 1>}
 dyn_hair_r_MSTR {parent=head, loc=<10.0791 -1.0364 6.0327>, rot=<0.1315 -0.5972 0.7911 0.0152>}
 dyn_hair_r_00 {parent=dyn_hair_r_MSTR, loc=<0.5 0 0>, rot=<-0.0006 0.0016 -0.0049 -1>}
 dyn_hair_r_01 {parent=dyn_hair_r_00, loc=<6.1151 0 0>, rot=<-0.0063 0.0224 -0.0723 -0.9971>}
 dyn_hair_r_02 {parent=dyn_hair_r_01, loc=<7.6839 0 0>, rot=<-0.0459 0.3353 -0.0465 -0.9398>}
 dyn_hair_r_03 {parent=dyn_hair_r_02, loc=<7.9495 0 0>, rot=<0.1523 -0.2716 -0.0403 -0.9494>}
 dyn_hair_r_04 {parent=dyn_hair_r_03, loc=<8.1126 0 0>, rot=<0.0022 -0.1842 -0.0117 -0.9828>}
 dyn_hair_r_05 {parent=dyn_hair_r_04, loc=<13.7322 0 0>, rot=<0 0 0 1>}
 attach_backpack {parent=spine_05, loc=<-0.123 -13.473 0>, rot=<-0.0093 -0.707 0.0093 -0.707>}
 clavicle_l {parent=spine_05, loc=<5.4831 -0.1053 -3.2946>, rot=<0.0465 -0.7289 -0.009 -0.6829>}
 upperarm_l {parent=clavicle_l, loc=<9.8658 -1.2136 -0.4105>, rot=<0.0164 -0.3245 -0.0269 -0.9454>}
 upperarm_twist_01_l {parent=upperarm_l, loc=<7.2135 -0.1134 0>, rot=<0 0 0 1>}
 deform_upperarm_bk_base_l {parent=upperarm_twist_01_l, loc=<-6.4053 -4.3746 0.314>, rot=<0 0 0 -1>}
 deform_upperarm_bk_l {parent=deform_upperarm_bk_base_l, loc=<0 0 0>, rot=<0 0 0 1>}
 deform_upperarm_fr_base_l {parent=upperarm_twist_01_l, loc=<-5.6312 5.092 0.3354>, rot=<0 0 0 -1>}
 deform_upperarm_up_base_l {parent=upperarm_twist_01_l, loc=<-8.891 0.4825 3.9517>, rot=<0 0 0 -1>}
 deform_upperarm_up_l {parent=deform_upperarm_up_base_l, loc=<0 0 0>, rot=<0 0 0 1>}
 extra_upperarm_l {parent=upperarm_twist_01_l, loc=<-7.2135 0.1134 0>, rot=<-0.0004 0.0034 0.0158 -0.9999>}
 extra_clavicle_l {parent=clavicle_l, loc=<0 0 0>, rot=<0 0 0 1>}
 clavicle_r {parent=spine_05, loc=<5.4831 -0.1053 3.2946>, rot=<0.7289 0.0465 -0.6829 0.009>}
 upperarm_r {parent=clavicle_r, loc=<-9.8658 1.2136 0.4105>, rot=<0.0164 -0.3245 -0.0269 -0.9454>}
 upperarm_twist_01_r {parent=upperarm_r, loc=<-7.2135 0.1134 0>, rot=<0 0 0 -1>}
 deform_upperarm_bk_base_r {parent=upperarm_twist_01_r, loc=<6.4053 4.3746 -0.314>, rot=<0 0 0 -1>}
 deform_upperarm_bk_r {parent=deform_upperarm_bk_base_r, loc=<0 0 0>, rot=<0 0 0 1>}
 deform_upperarm_up_base_r {parent=upperarm_twist_01_r, loc=<8.891 -0.4825 -3.9517>, rot=<0 0 0 -1>}
 deform_upperarm_up_r {parent=deform_upperarm_up_base_r, loc=<0 0 0>, rot=<0 0 0 1>}
 extra_upperarm_r {parent=upperarm_twist_01_r, loc=<5.5372 -0.0605 -0.0115>, rot=<0.0018 0.012 0.1806 -0.9835>}
 dyn_skirt_bk_a_base_l {parent=pelvis, loc=<7.0462 -6.4878 -8.0885>, rot=<0.9235 0.1075 0.069 -0.3616>}
 dyn_skirt_bk_a_l {parent=dyn_skirt_bk_a_base_l, loc=<0 0 0>, rot=<0 0 0 1>}
 dyn_skirt_bk_c_l {parent=dyn_skirt_bk_a_l, loc=<-3.3764 0.1048 -0.2511>, rot=<0 0 0 1>}
 dyn_skirt_bk_a_base_r {parent=pelvis, loc=<6.6858 -6.7724 8.088>, rot=<0.1075 -0.9235 0.3616 0.069>}
 dyn_skirt_bk_a_r {parent=dyn_skirt_bk_a_base_r, loc=<0 0 0>, rot=<0 0 0 1>}
 dyn_skirt_bk_c_r {parent=dyn_skirt_bk_a_r, loc=<2.1552 0.0424 0.1647>, rot=<0 0 0 1>}
 dyn_skirt_bk_a_base {parent=pelvis, loc=<6.4686 -7.9331 0.0859>, rot=<0.094 -0.9956 0.0002 -0.0018>}
 dyn_skirt_bk_a {parent=dyn_skirt_bk_a_base, loc=<0 0 0>, rot=<0 0 0 1>}
 extra_back_pelvis_m {parent=pelvis, loc=<-0.9356 -10.1249 -0.0221>, rot=<-0.0127 -0.707 0.0127 -0.707>}
]
variables [Is skydive {type=bool, default=False}]
