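# Physics asset for the dynamic hair of the F_MED_PastelGlazeGrain female character's face accessory.
# Left, right and back six-capsule chains hang from master bodies on the head, with limited swing and very tight twist between segments.
# Free-angle cross constraints tie both side chains to the back chain mid-length and at the tips, and the back chain to attach_cape.
# Capsules on head, neck, spine, pelvis, shoulders, upper arms and a cape collider supply body shapes.
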PhysicsAsset /BRCosmetics/Characters/Player/Female/Medium/Bodies/F_MED_PastelGlazeGrain/Meshes/Parts/F_MED_PastelGlazeGrain_FaceAcc_Dyn_Hair
bodies {
 attach_cape [{shapeType=Capsule, length=0.505, radius=0.505, shapeName=attach_cape_capsule, rotation=<rot 0 -90 0>}]
 clavicle_l [{
   shapeType=Capsule
   length=3.25
   radius=5.625
   shapeName=clavicle_l_capsule
   center=<6.7755 -0.4278 -0.9567>
   rotation=<rot 86.2624 -0.0026 -0.0026>
  }]
 clavicle_r [{
   shapeType=Capsule
   length=3.25
   radius=5.625
   shapeName=clavicle_l_capsule
   center=<-6.7755 0.4288 0.9567>
   rotation=<rot 86.2624 -0.0026 179.9974>
  }]
 dyn_cape_collider [{shapeType=Capsule, length=8, radius=12, shapeName=spine_05_capsule, center=<0 9 0.2845>, rotation=<rot 90 0 0>}]
 dyn_hair_bk_00 [{shapeType=Capsule, length=3.75, radius=6.5, shapeName=dyn_hair_bk_01_capsule, center=<5 0 4>, rotation=<rot 0 0 90>}]
 dyn_hair_bk_01 [{shapeType=Capsule, length=3.25, radius=7, shapeName=dyn_hair_bk_01_capsule, center=<6 0 4>, rotation=<rot 0 0 90>}]
 dyn_hair_bk_02 [{shapeType=Capsule, length=3.125, radius=7, shapeName=dyn_hair_bk_02_capsule, center=<10 0 4>, rotation=<rot 0 0 90>}]
 dyn_hair_bk_03 [{shapeType=Capsule, length=5.375, radius=6.375, shapeName=dyn_hair_bk_03_capsule, center=<6 0 4>, rotation=<rot 4.9999 0 90>}]
 dyn_hair_bk_04 [{shapeType=Capsule, length=3.875, radius=7.75, shapeName=dyn_hair_bk_04_capsule, center=<6 0 4>, rotation=<rot 0 0 90>}]
 dyn_hair_bk_05 [{shapeType=Capsule, length=6, radius=5, shapeName=dyn_hair_bk_05_capsule, center=<5.086 0 -1.3893>, rotation=<rot 90 90 0>}]
 dyn_hair_bk_MSTR [{shapeType=Capsule, length=0.505, radius=0.505, shapeName=dyn_hair_bk_MSTR_capsule, rotation=<rot 0 -90 0>}]
 dyn_hair_l_00 [{shapeType=Capsule, length=3.25, radius=3.875, shapeName=dyn_hair_l_01_capsule, center=<4 1 4>, rotation=<rot 90 0 0>}]
 dyn_hair_l_01 [{shapeType=Capsule, length=2.625, radius=4.5, shapeName=dyn_hair_l_01_capsule, center=<4 0 4>, rotation=<rot 90 0 0>}]
 dyn_hair_l_02 [{shapeType=Capsule, length=2.5, radius=6.5, shapeName=dyn_hair_l_02_capsule, center=<6.7355 -0.3581 1.5545>, rotation=<rot -0.0005 0 99.9999>}]
 dyn_hair_l_03 [{shapeType=Capsule, length=4.125, radius=6.75, shapeName=dyn_hair_l_03_capsule, center=<5 0 3.2466>, rotation=<rot 0 0 90>}]
 dyn_hair_l_04 [{shapeType=Capsule, length=4.625, radius=5.625, shapeName=dyn_hair_l_04_capsule, center=<5 0.0604 3.3091>, rotation=<rot -0.0002 0 85>}]
 dyn_hair_l_05 [{shapeType=Capsule, length=6, radius=5, shapeName=dyn_hair_l_05_capsule, center=<6.387 0 -2.7549>, rotation=<rot 90 90 0>}]
 dyn_hair_l_MSTR [{shapeType=Capsule, length=0.505, radius=0.505, shapeName=dyn_hair_l_MSTR_capsule, rotation=<rot 0 -90 0>}]
 dyn_hair_r_00 [{shapeType=Capsule, length=3.375, radius=3.625, shapeName=dyn_hair_r_01_capsule, center=<5 0 3>, rotation=<rot 90 0 0>}]
 dyn_hair_r_01 [{
   shapeType=Capsule
   length=3.625
   radius=7.875
   shapeName=dyn_hair_r_01_capsule
   center=<5.231 0 0.36>
   rotation=<rot 85.0002 179.9998 179.9998>
  }]
 dyn_hair_r_02 [{
   shapeType=Capsule
   length=4.375
   radius=5.625
   shapeName=dyn_hair_r_02_capsule
   center=<6.0593 -0.1093 0.7506>
   rotation=<rot -0.0002 -0.0001 95>
  }]
 dyn_hair_r_03 [{shapeType=Capsule, length=7.125, radius=6, shapeName=dyn_hair_r_03_capsule, center=<5 -2.4789 1.3098>, rotation=<rot -0.0005 0 90>}]
 dyn_hair_r_04 [{
   shapeType=Capsule
   length=13.125
   radius=6.25
   shapeName=dyn_hair_r_04_capsule
   center=<6.2971 0.0885 2.5037>
   rotation=<rot -0.0002 0 86.6144>
  }]
 dyn_hair_r_05 [{shapeType=Capsule, length=11.5, radius=5, shapeName=dyn_hair_r_05_capsule, center=<5.1428 0 -0.8216>, rotation=<rot 90 180 90>}]
 dyn_hair_r_MSTR [{shapeType=Capsule, length=0.505, radius=0.505, shapeName=dyn_hair_r_MSTR_capsule, rotation=<rot 0 -90 0>}]
 head [{shapeType=Capsule, length=6.1965, radius=6.5617, shapeName=head_capsule, center=<2.5967 2.5691 0>, rotation=<rot 90 0 0>}]
 neck_01 [{shapeType=Capsule, length=4.1965, radius=6.9367, shapeName=head_capsule, center=<2.5967 2.0576 0>, rotation=<rot 90 0 0>}]
 pelvis [{shapeType=Capsule, length=6.125, radius=13.875, shapeName=spine_05_capsule, center=<-1.553 -0.8492 0>}]
 spine_02 [{shapeType=Capsule, length=3.125, radius=13.875, shapeName=spine_05_capsule, center=<0.1876 6.0025 0>}]
 spine_04 [{shapeType=Capsule, length=4.5, radius=12, shapeName=spine_05_capsule, center=<1.0391 2.1461 0>}]
 spine_05 [{shapeType=Capsule, length=4.25, radius=12, shapeName=spine_05_capsule, center=<1.1451 2.0175 0>}]
 upperarm_l [{shapeType=Capsule, length=19, radius=5.625, shapeName=clavicle_l_capsule, center=<13.7543 0 -0.5227>, rotation=<rot 90 0 0>}]
 upperarm_r [{shapeType=Capsule, length=21, radius=5.625, shapeName=clavicle_l_capsule, center=<-13.7543 0 0.5227>, rotation=<rot 90 0 180>}]
}
constraints [
 {
  bone1Name=dyn_hair_l_MSTR
  bone2Name=head
  swing1LimitDegrees=45
  swing1Motion=Free
  swing2LimitDegrees=45
  swing2Motion=Free
  twistLimitDegrees=45
  twistMotion=Free
 }
 {bone1Name=dyn_hair_l_01, bone2Name=dyn_hair_l_00, swing1LimitDegrees=20, swing2LimitDegrees=20, twistLimitDegrees=2}
 {bone1Name=dyn_hair_l_02, bone2Name=dyn_hair_l_01, swing1LimitDegrees=20, swing2LimitDegrees=40, twistLimitDegrees=2}
 {bone1Name=dyn_hair_l_03, bone2Name=dyn_hair_l_02, swing1LimitDegrees=20, swing2LimitDegrees=30, twistLimitDegrees=2}
 {bone1Name=dyn_hair_l_04, bone2Name=dyn_hair_l_03, swing1LimitDegrees=20, swing2LimitDegrees=30, twistLimitDegrees=2}
 {
  bone1Name=dyn_hair_bk_MSTR
  bone2Name=head
  swing1LimitDegrees=45
  swing1Motion=Free
  swing2LimitDegrees=45
  swing2Motion=Free
  twistLimitDegrees=45
  twistMotion=Free
 }
 {bone1Name=dyn_hair_bk_01, bone2Name=dyn_hair_bk_00, swing1LimitDegrees=20, swing2LimitDegrees=20, twistLimitDegrees=2}
 {bone1Name=dyn_hair_bk_02, bone2Name=dyn_hair_bk_01, swing1LimitDegrees=20, swing2LimitDegrees=20, twistLimitDegrees=2}
 {bone1Name=dyn_hair_bk_03, bone2Name=dyn_hair_bk_02, swing1LimitDegrees=20, swing2LimitDegrees=30, twistLimitDegrees=2}
 {bone1Name=dyn_hair_bk_04, bone2Name=dyn_hair_bk_03, swing1LimitDegrees=20, swing2LimitDegrees=30, twistLimitDegrees=2}
 {
  bone1Name=dyn_hair_r_MSTR
  bone2Name=head
  swing1LimitDegrees=45
  swing1Motion=Free
  swing2LimitDegrees=45
  swing2Motion=Free
  twistLimitDegrees=45
  twistMotion=Free
 }
 {bone1Name=dyn_hair_r_01, bone2Name=dyn_hair_r_00, swing1LimitDegrees=10, swing2LimitDegrees=20, twistLimitDegrees=2}
 {bone1Name=dyn_hair_r_02, bone2Name=dyn_hair_r_01, swing1LimitDegrees=15, swing2LimitDegrees=40, twistLimitDegrees=2}
 {bone1Name=dyn_hair_r_03, bone2Name=dyn_hair_r_02, swing1LimitDegrees=20, swing2LimitDegrees=30, twistLimitDegrees=2}
 {bone1Name=dyn_hair_r_04, bone2Name=dyn_hair_r_03, swing1LimitDegrees=20, swing2LimitDegrees=30, twistLimitDegrees=2}
 {
  bone1Name=head
  bone2Name=neck_01
  swing1LimitDegrees=45
  swing1Motion=Free
  swing2LimitDegrees=45
  swing2Motion=Free
  twistLimitDegrees=45
  twistMotion=Free
 }
 {
  bone1Name=spine_05
  bone2Name=spine_04
  swing1LimitDegrees=45
  swing1Motion=Free
  swing2LimitDegrees=45
  swing2Motion=Free
  twistLimitDegrees=45
  twistMotion=Free
 }
 {
  bone1Name=clavicle_r
  bone2Name=spine_05
  swing1LimitDegrees=45
  swing1Motion=Free
  swing2LimitDegrees=45
  swing2Motion=Free
  twistLimitDegrees=45
  twistMotion=Free
 }
 {
  bone1Name=neck_01
  bone2Name=spine_05
  swing1LimitDegrees=45
  swing1Motion=Free
  swing2LimitDegrees=45
  swing2Motion=Free
  twistLimitDegrees=45
  twistMotion=Free
 }
 {
  bone1Name=upperarm_r
  bone2Name=clavicle_r
  swing1LimitDegrees=45
  swing1Motion=Free
  swing2LimitDegrees=45
  swing2Motion=Free
  twistLimitDegrees=45
  twistMotion=Free
 }
 {
  bone1Name=spine_04
  bone2Name=spine_02
  swing1LimitDegrees=45
  swing1Motion=Free
  swing2LimitDegrees=45
  swing2Motion=Free
  twistLimitDegrees=45
  twistMotion=Free
 }
 {
  bone1Name=spine_02
  bone2Name=pelvis
  swing1LimitDegrees=45
  swing1Motion=Free
  swing2LimitDegrees=45
  swing2Motion=Free
  twistLimitDegrees=45
  twistMotion=Free
 }
 {bone1Name=dyn_hair_r_00, bone2Name=dyn_hair_r_MSTR, swing1LimitDegrees=20, swing2LimitDegrees=20, twistLimitDegrees=2}
 {bone1Name=dyn_hair_bk_00, bone2Name=dyn_hair_bk_MSTR, swing1LimitDegrees=20, swing2LimitDegrees=10, twistLimitDegrees=2}
 {bone1Name=dyn_hair_l_00, bone2Name=dyn_hair_l_MSTR, swing1LimitDegrees=20, swing2LimitDegrees=20, twistLimitDegrees=2}
 {
  bone1Name=clavicle_l
  bone2Name=spine_05
  swing1LimitDegrees=45
  swing1Motion=Free
  swing2LimitDegrees=45
  swing2Motion=Free
  twistLimitDegrees=45
  twistMotion=Free
 }
 {bone1Name=upperarm_l, bone2Name=clavicle_l, swing1LimitDegrees=20, swing2LimitDegrees=15, twistLimitDegrees=15}
 {
  bone1Name=dyn_hair_r_03
  bone2Name=dyn_hair_bk_02
  swing1LimitDegrees=45
  swing1Motion=Free
  swing2LimitDegrees=45
  swing2Motion=Free
  twistLimitDegrees=45
  twistMotion=Free
 }
 {
  bone1Name=dyn_hair_l_03
  bone2Name=dyn_hair_bk_02
  swing1LimitDegrees=45
  swing1Motion=Free
  swing2LimitDegrees=45
  swing2Motion=Free
  twistLimitDegrees=45
  twistMotion=Free
 }
 {bone1Name=dyn_hair_r_05, bone2Name=dyn_hair_r_04, swing1LimitDegrees=10, swing2LimitDegrees=10, twistLimitDegrees=3}
 {bone1Name=dyn_hair_bk_05, bone2Name=dyn_hair_bk_04, swing1LimitDegrees=10, swing2LimitDegrees=10, twistLimitDegrees=3}
 {bone1Name=dyn_hair_l_05, bone2Name=dyn_hair_l_04, swing1LimitDegrees=10, swing2LimitDegrees=10, twistLimitDegrees=3}
 {
  bone1Name=dyn_hair_r_05
  bone2Name=dyn_hair_bk_05
  swing1LimitDegrees=45
  swing1Motion=Free
  swing2LimitDegrees=45
  swing2Motion=Free
  twistLimitDegrees=45
  twistMotion=Free
 }
 {
  bone1Name=dyn_hair_l_05
  bone2Name=dyn_hair_bk_05
  swing1LimitDegrees=45
  swing1Motion=Free
  swing2LimitDegrees=45
  swing2Motion=Free
  twistLimitDegrees=45
  twistMotion=Free
 }
 {
  bone1Name=attach_cape
  bone2Name=spine_05
  swing1LimitDegrees=45
  swing1Motion=Free
  swing2LimitDegrees=45
  swing2Motion=Free
  twistLimitDegrees=45
  twistMotion=Free
 }
 {
  bone1Name=dyn_hair_bk_03
  bone2Name=attach_cape
  swing1LimitDegrees=45
  swing1Motion=Free
  swing2LimitDegrees=45
  swing2Motion=Free
  twistLimitDegrees=45
  twistMotion=Free
 }
 {
  bone1Name=dyn_cape_collider
  bone2Name=spine_05
  swing1LimitDegrees=45
  swing1Motion=Free
  swing2LimitDegrees=45
  swing2Motion=Free
  twistLimitDegrees=45
  twistMotion=Free
 }
]
